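# Physics asset for the dynamic skirt on the Golden Skeleton male medium player body, built entirely from capsule bodies.
# Front, side and back cloth bone chains hang from the pelvis, front and sides via MSTR anchor bones, with swing and twist limits along each chain.
# Pelvis, thigh and calf capsules form the lower-body shapes, joined by free angular constraints.
# Large root and zero-length crouch collision capsules add broad volumes.
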PhysicsAsset /BRCosmetics/Characters/Player/Male/Medium/Bodies/M_MED_The_Golden_Skeleton/Meshes/M_MED_The_Golden_Skeleton_Dyn_skirt
bodies {
 calf_l [{shapeType=Capsule, length=37.6245, radius=7.1301, center=<-25.5503 -1.4803 -0.0001>, rotation=<rot 90 -90 -81.1481>}]
 calf_r [{shapeType=Capsule, length=35.2546, radius=8, center=<25.8649 1.4316 0.0001>, rotation=<rot 90 -90 -81.1481>}]
 dyn_cloth_bk_01 [{
   shapeType=Capsule
   length=11.5
   radius=3.25
   shapeName=dyn_cloth_bk_01_capsule
   center=<9.342 0 -0.1724>
   rotation=<rot -89.536 179.9971 -179.9971>
  }]
 dyn_cloth_bk_02 [{shapeType=Capsule, length=6.5, radius=8, shapeName=dyn_cloth_bk_02_capsule, center=<5 0 3>, rotation=<rot 0 0 90>}]
 dyn_cloth_bk_03 [{shapeType=Capsule, length=10.5, radius=8, shapeName=dyn_cloth_bk_03_capsule, center=<5 0 3>, rotation=<rot 0 0 90>}]
 dyn_cloth_bk_04 [{shapeType=Capsule, length=18.25, radius=8, shapeName=dyn_cloth_bk_04_capsule, center=<5 3 3>, rotation=<rot 0 0 95>}]
 dyn_cloth_bk_05 [{shapeType=Capsule, length=10.25, radius=8, shapeName=dyn_cloth_bk_05_capsule, center=<5 6 3>, rotation=<rot 0 0 95>}]
 dyn_cloth_bk_l_01 [{
   shapeType=Capsule
   length=11.5
   radius=3.25
   shapeName=dyn_cloth_bk_l_01_capsule
   center=<9.342 0 -0.1687>
   rotation=<rot -89.5771 179.9941 -179.9941>
  }]
 dyn_cloth_bk_l_02 [{
   shapeType=Capsule
   length=11.5
   radius=3.25
   shapeName=dyn_cloth_bk_l_02_capsule
   center=<9.342 0 -0.1687>
   rotation=<rot -89.5771 179.9941 -179.9941>
  }]
 dyn_cloth_bk_l_03 [{
   shapeType=Capsule
   length=11.5
   radius=3.25
   shapeName=dyn_cloth_bk_l_03_capsule
   center=<9.342 0 -0.1687>
   rotation=<rot -89.5771 179.9941 -179.9941>
  }]
 dyn_cloth_bk_l_04 [{
   shapeType=Capsule
   length=11.5
   radius=3.25
   shapeName=dyn_cloth_bk_l_04_capsule
   center=<9.342 0 -0.1687>
   rotation=<rot -89.5771 179.9941 -179.9941>
  }]
 dyn_cloth_bk_l_05 [{
   shapeType=Capsule
   length=11.5
   radius=3.25
   shapeName=dyn_cloth_bk_l_05_capsule
   center=<9.342 0 -0.1687>
   rotation=<rot -89.5771 179.9941 -179.9941>
  }]
 dyn_cloth_bk_r_01 [{
   shapeType=Capsule
   length=9.75
   radius=3.25
   shapeName=dyn_cloth_bk_r_01_capsule
   center=<5.9658 -0.3783 0.9599>
   rotation=<rot -89.2718 -3.6319 0.0006>
  }]
 dyn_cloth_bk_r_02 [{
   shapeType=Capsule
   length=9.75
   radius=3.25
   shapeName=dyn_cloth_bk_r_02_capsule
   center=<5.9658 -0.3783 0.9599>
   rotation=<rot -89.2718 -3.6319 0.0006>
  }]
 dyn_cloth_bk_r_03 [{
   shapeType=Capsule
   length=9.75
   radius=3.25
   shapeName=dyn_cloth_bk_r_03_capsule
   center=<5.9658 -0.3783 0.9599>
   rotation=<rot -89.2718 -3.6319 0.0006>
  }]
 dyn_cloth_fr_l_01 [{
   shapeType=Capsule
   length=7.5
   radius=3.5
   shapeName=dyn_cloth_fr_l_01_capsule
   center=<6.6495 0 -0.2636>
   rotation=<rot -87.7298 -179.9991 179.9991>
  }]
 dyn_cloth_fr_l_02 [{
   shapeType=Capsule
   length=11.3629
   radius=2.7203
   shapeName=dyn_cloth_fr_l_02_capsule
   center=<6.6495 0 -0.2635>
   rotation=<rot -87.7306 -179.9991 179.9991>
  }]
 dyn_cloth_fr_l_03 [{
   shapeType=Capsule
   length=11.3629
   radius=2.7203
   shapeName=dyn_cloth_fr_l_03_capsule
   center=<6.6495 0 -0.2635>
   rotation=<rot -87.7306 -179.9991 179.9991>
  }]
 dyn_cloth_fr_l_MSTR [{shapeType=Capsule, length=0.5, radius=0.5}]
 dyn_cloth_fr_r_01 [{
   shapeType=Capsule
   length=7.5
   radius=3.5
   shapeName=dyn_cloth_fr_r_01_capsule
   center=<6.2139 -0.0005 0.2559>
   rotation=<rot -86.9747 179.9999 -179.9999>
  }]
 dyn_cloth_fr_r_02 [{
   shapeType=Capsule
   length=8.77
   radius=2.9171
   shapeName=dyn_cloth_fr_r_02_capsule
   center=<6.1831 -0.0005 -0.3265>
   rotation=<rot -86.9775 179.9999 -179.9999>
  }]
 dyn_cloth_fr_r_03 [{
   shapeType=Capsule
   length=9.02
   radius=2.6671
   shapeName=dyn_cloth_fr_r_03_capsule
   center=<7.7018 -0.0004 -0.4067>
   rotation=<rot -86.9775 179.9999 -179.9999>
  }]
 dyn_cloth_fr_r_MSTR [{shapeType=Capsule, length=0.5, radius=0.5}]
 dyn_cloth_l_01 [{shapeType=Capsule, length=10.25, radius=4.75, shapeName=dyn_cloth_l_01_capsule, center=<3 -5 4>, rotation=<rot -0.0003 0.0006 99.9997>}]
 dyn_cloth_l_02 [{shapeType=Capsule, length=11.5, radius=5, shapeName=dyn_cloth_l_02_capsule, center=<5 -5 4>, rotation=<rot -0.0003 0.0006 99.9997>}]
 dyn_cloth_l_03 [{shapeType=Capsule, length=11.75, radius=7.25, shapeName=dyn_cloth_l_03_capsule, center=<5 -5 4>, rotation=<rot -0.0003 0.0007 89.9997>}]
 dyn_cloth_l_04 [{shapeType=Capsule, length=15.25, radius=7.75, shapeName=dyn_cloth_l_04_capsule, center=<3 -5 5>, rotation=<rot -0.0003 0.0007 89.9997>}]
 dyn_cloth_l_05 [{shapeType=Capsule, length=13.75, radius=9.5, shapeName=dyn_cloth_l_05_capsule, center=<3 -5 5>, rotation=<rot -0.0001 0.0003 -102.5241>}]
 dyn_cloth_l_06 [{
   shapeType=Capsule
   length=10.5
   radius=11.75
   shapeName=dyn_cloth_l_06_capsule
   center=<7.1743 -5 5>
   rotation=<rot -0.0001 0.0003 -102.5241>
  }]
 dyn_cloth_l_MSTR [{shapeType=Capsule, length=0.5, radius=0.5}]
 dyn_cloth_r_01 [{
   shapeType=Capsule
   length=6.25
   radius=2
   shapeName=dyn_cloth_r_01_capsule
   center=<3.3236 -0.2098 0.8854>
   rotation=<rot -89.2788 -3.6236 0.0005>
  }]
 dyn_cloth_r_02 [{shapeType=Capsule, length=8, radius=3.75, shapeName=dyn_cloth_r_02_capsule, center=<3 0 3>, rotation=<rot -0.0001 -0.0002 70.0001>}]
 dyn_cloth_r_03 [{shapeType=Capsule, length=9.25, radius=5.75, shapeName=dyn_cloth_r_03_capsule, center=<3 0 4>, rotation=<rot -0.0001 -0.0003 70.0001>}]
 dyn_cloth_r_MSTR [{shapeType=Capsule, length=0.5, radius=0.5}]
 dyn_crouch_collision [{shapeType=Capsule, length=0, radius=39.8749, shapeName=dyn_crouch_collision_capsule, center=<-0.0002 -2.3087 -11.4096>}]
 pelvis [{shapeType=Capsule, length=6.5601, radius=8.0964, center=<-5.788 0.6928 0.0005>}]
 root [{shapeType=Capsule, length=39.5, radius=85.75, shapeName=root_capsule, center=<19.4912 2.9533 -88.856>}]
 thigh_l [{shapeType=Capsule, length=30.6836, radius=6, center=<-18.7531 0.0004 0.2337>, rotation=<rot 87.3858 180 180>}]
 thigh_r [{shapeType=Capsule, length=29.7749, radius=6.9087, center=<18.7531 -0.0004 -0.2337>, rotation=<rot 87.3858 180 180>}]
}
constraints [
 {bone1Name=dyn_cloth_fr_r_01, bone2Name=dyn_cloth_fr_r_MSTR, swing1LimitDegrees=30, swing2LimitDegrees=30, twistLimitDegrees=5}
 {bone1Name=dyn_cloth_fr_r_02, bone2Name=dyn_cloth_fr_r_01, swing1LimitDegrees=30, swing2LimitDegrees=30, twistLimitDegrees=5}
 {bone1Name=dyn_cloth_fr_r_03, bone2Name=dyn_cloth_fr_r_02, swing1LimitDegrees=30, swing2LimitDegrees=30, twistLimitDegrees=5}
 {bone1Name=dyn_cloth_fr_l_01, bone2Name=dyn_cloth_fr_l_MSTR, swing1LimitDegrees=30, swing2LimitDegrees=30, twistLimitDegrees=5}
 {bone1Name=dyn_cloth_fr_l_02, bone2Name=dyn_cloth_fr_l_01, swing1LimitDegrees=30, swing2LimitDegrees=30, twistLimitDegrees=5}
 {bone1Name=dyn_cloth_fr_l_03, bone2Name=dyn_cloth_fr_l_02, swing1LimitDegrees=30, swing2LimitDegrees=30, twistLimitDegrees=5}
 {bone1Name=dyn_cloth_bk_l_02, bone2Name=dyn_cloth_bk_l_01, swing1LimitDegrees=60, swing2LimitDegrees=60, twistLimitDegrees=5}
 {bone1Name=dyn_cloth_bk_l_03, bone2Name=dyn_cloth_bk_l_02, swing1LimitDegrees=60, swing2LimitDegrees=60, twistLimitDegrees=5}
 {bone1Name=dyn_cloth_bk_l_04, bone2Name=dyn_cloth_bk_l_03, swing1LimitDegrees=90, swing2LimitDegrees=90, twistLimitDegrees=15}
 {bone1Name=dyn_cloth_bk_l_05, bone2Name=dyn_cloth_bk_l_04, swing1LimitDegrees=90, swing2LimitDegrees=90, twistLimitDegrees=15}
 {bone1Name=dyn_cloth_l_02, bone2Name=dyn_cloth_l_01, swing1LimitDegrees=10, swing2LimitDegrees=45, twistLimitDegrees=5}
 {bone1Name=dyn_cloth_l_03, bone2Name=dyn_cloth_l_02, swing1LimitDegrees=10, swing2LimitDegrees=60, twistLimitDegrees=5}
 {bone1Name=dyn_cloth_l_04, bone2Name=dyn_cloth_l_03, swing1LimitDegrees=10, swing2LimitDegrees=60, twistLimitDegrees=15}
 {bone1Name=dyn_cloth_l_05, bone2Name=dyn_cloth_l_04, swing1LimitDegrees=10, swing2LimitDegrees=60, twistLimitDegrees=15}
 {bone1Name=dyn_cloth_l_06, bone2Name=dyn_cloth_l_05, swing1LimitDegrees=10, swing2LimitDegrees=60, twistLimitDegrees=15}
 {bone1Name=dyn_cloth_bk_02, bone2Name=dyn_cloth_bk_01, swing1LimitDegrees=60, swing2LimitDegrees=60, twistLimitDegrees=5}
 {bone1Name=dyn_cloth_bk_03, bone2Name=dyn_cloth_bk_02, swing1LimitDegrees=60, swing2LimitDegrees=60, twistLimitDegrees=5}
 {bone1Name=dyn_cloth_bk_04, bone2Name=dyn_cloth_bk_03, swing1LimitDegrees=60, swing2LimitDegrees=60, twistLimitDegrees=5}
 {bone1Name=dyn_cloth_bk_05, bone2Name=dyn_cloth_bk_04, swing1LimitDegrees=60, swing2LimitDegrees=60, twistLimitDegrees=5}
 {bone1Name=dyn_cloth_bk_r_02, bone2Name=dyn_cloth_bk_r_01, swing1LimitDegrees=60, swing2LimitDegrees=60, twistLimitDegrees=5}
 {bone1Name=dyn_cloth_bk_r_03, bone2Name=dyn_cloth_bk_r_02, swing1LimitDegrees=60, swing2LimitDegrees=60, twistLimitDegrees=5}
 {bone1Name=dyn_cloth_r_02, bone2Name=dyn_cloth_r_01, swing1LimitDegrees=30, swing2LimitDegrees=30, twistLimitDegrees=5}
 {bone1Name=dyn_cloth_r_03, bone2Name=dyn_cloth_r_02, swing1LimitDegrees=30, swing2LimitDegrees=30, twistLimitDegrees=5}
 {
  bone1Name=dyn_cloth_fr_r_MSTR
  bone2Name=pelvis
  swing1LimitDegrees=45
  swing1Motion=Free
  swing2LimitDegrees=45
  swing2Motion=Free
  twistLimitDegrees=45
  twistMotion=Free
 }
 {
  bone1Name=dyn_cloth_fr_l_MSTR
  bone2Name=pelvis
  swing1LimitDegrees=45
  swing1Motion=Free
  swing2LimitDegrees=45
  swing2Motion=Free
  twistLimitDegrees=45
  twistMotion=Free
 }
 {bone1Name=dyn_cloth_bk_l_01, bone2Name=pelvis, swing1LimitDegrees=60, swing2LimitDegrees=60, twistLimitDegrees=5}
 {bone1Name=dyn_cloth_l_01, bone2Name=dyn_cloth_l_MSTR, swing1LimitDegrees=0.1, swing2LimitDegrees=45, twistLimitDegrees=0.1}
 {bone1Name=dyn_cloth_bk_01, bone2Name=pelvis, swing1LimitDegrees=60, swing2LimitDegrees=60, twistLimitDegrees=5}
 {bone1Name=dyn_cloth_bk_r_01, bone2Name=pelvis, swing1LimitDegrees=60, swing2LimitDegrees=60, twistLimitDegrees=5}
 {bone1Name=dyn_cloth_r_01, bone2Name=dyn_cloth_r_MSTR, swing1LimitDegrees=30, swing2LimitDegrees=30, twistLimitDegrees=5}
 {
  bone1Name=thigh_l
  bone2Name=pelvis
  swing1LimitDegrees=45
  swing1Motion=Free
  swing2LimitDegrees=45
  swing2Motion=Free
  twistLimitDegrees=45
  twistMotion=Free
 }
 {
  bone1Name=calf_l
  bone2Name=thigh_l
  swing1LimitDegrees=45
  swing1Motion=Free
  swing2LimitDegrees=45
  swing2Motion=Free
  twistLimitDegrees=45
  twistMotion=Free
 }
 {
  bone1Name=thigh_r
  bone2Name=pelvis
  swing1LimitDegrees=45
  swing1Motion=Free
  swing2LimitDegrees=45
  swing2Motion=Free
  twistLimitDegrees=45
  twistMotion=Free
 }
 {
  bone1Name=calf_r
  bone2Name=thigh_r
  swing1LimitDegrees=45
  swing1Motion=Free
  swing2LimitDegrees=45
  swing2Motion=Free
  twistLimitDegrees=45
  twistMotion=Free
 }
 {
  bone1Name=pelvis
  bone2Name=root
  swing1LimitDegrees=45
  swing1Motion=Free
  swing2LimitDegrees=45
  swing2Motion=Free
  twistLimitDegrees=45
  twistMotion=Free
 }
 {
  bone1Name=dyn_cloth_l_MSTR
  bone2Name=pelvis
  swing1LimitDegrees=45
  swing1Motion=Free
  swing2LimitDegrees=45
  swing2Motion=Free
  twistLimitDegrees=45
  twistMotion=Free
 }
 {
  bone1Name=dyn_cloth_r_MSTR
  bone2Name=pelvis
  swing1LimitDegrees=45
  swing1Motion=Free
  swing2LimitDegrees=45
  swing2Motion=Free
  twistLimitDegrees=45
  twistMotion=Free
 }
 {
  bone1Name=dyn_crouch_collision
  bone2Name=root
  swing1LimitDegrees=45
  swing1Motion=Free
  swing2LimitDegrees=45
  swing2Motion=Free
  twistLimitDegrees=45
  twistMotion=Free
 }
]
