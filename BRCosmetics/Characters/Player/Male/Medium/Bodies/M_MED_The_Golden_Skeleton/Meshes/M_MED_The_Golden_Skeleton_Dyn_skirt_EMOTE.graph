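# Physics asset for the Golden Skeleton medium male body's dynamic skirt, apparently the emote variant.
# Box and capsule chains form front, side and back panels, hung from the pelvis with limited swing and twist, the front and side ones via small MSTR anchor bodies.
# Pelvis, leg, root and crouch-collision capsules provide body shapes.
# Angularly free cross-constraints tie lower left-side and back segments together, and the two front panels to each other.
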PhysicsAsset /BRCosmetics/Characters/Player/Male/Medium/Bodies/M_MED_The_Golden_Skeleton/Meshes/M_MED_The_Golden_Skeleton_Dyn_skirt_EMOTE
bodies {
 calf_l [{shapeType=Capsule, length=37.9716, radius=9.7471, center=<-25.5503 -1.4803 -0.0001>, rotation=<rot 90 -90 -81.1481>}]
 calf_r [{shapeType=Capsule, length=35.2546, radius=8, center=<25.8649 1.4316 0.0001>, rotation=<rot 90 -90 -81.1481>}]
 dyn_cloth_bk_01 [{
   shapeType=Capsule
   length=11.5
   radius=3.25
   shapeName=dyn_cloth_bk_01_capsule
   center=<9.342 0 -0.1724>
   rotation=<rot -89.536 179.9971 -179.9971>
  }]
 dyn_cloth_bk_02 [{
   shapeType=Box
   extentX=13
   extentY=18
   extentZ=1
   shapeName=dyn_cloth_bk_02_box
   center=<5.5657 2.0035 0.0001>
   rotation=<rot -0.0573 0.2911 11.1232>
  }]
 dyn_cloth_bk_03 [{
   shapeType=Box
   extentX=15
   extentY=21.415
   extentZ=1
   shapeName=dyn_cloth_bk_03_box
   center=<7.1639 2.0035 0.0002>
   rotation=<rot 2.3834 -12.2226 10.8681>
  }]
 dyn_cloth_bk_04 [{
   shapeType=Box
   extentX=12
   extentY=26.3502
   extentZ=4.984
   shapeName=dyn_cloth_bk_04_box
   center=<5.083 1.2221 -2.0754>
   rotation=<rot 1.2029 -14.7579 4.5566>
  }]
 dyn_cloth_bk_05 [{
   shapeType=Box
   extentX=11.6123
   extentY=29.3198
   extentZ=5.8143
   shapeName=dyn_cloth_bk_05_box
   center=<3.2416 1.8438 -0.6294>
   rotation=<rot 1.9994 -18.2306 6.0468>
  }]
 dyn_cloth_bk_l_01 [{
   shapeType=Capsule
   length=11.5
   radius=3.25
   shapeName=dyn_cloth_bk_l_01_capsule
   center=<9.342 0 -0.1687>
   rotation=<rot -89.5771 179.9941 -179.9941>
  }]
 dyn_cloth_bk_l_02 [{
   shapeType=Capsule
   length=9.6714
   radius=5.0786
   shapeName=dyn_cloth_bk_l_02_capsule
   center=<9.342 0 -0.1687>
   rotation=<rot -89.5771 179.9935 179.8468>
  }]
 dyn_cloth_bk_l_03 [{
   shapeType=Capsule
   length=7.9232
   radius=6.8268
   shapeName=dyn_cloth_bk_l_03_capsule
   center=<9.342 0 -0.1687>
   rotation=<rot -89.5771 179.9941 -179.9941>
  }]
 dyn_cloth_bk_l_04 [{
   shapeType=Capsule
   length=7.9232
   radius=6.8268
   shapeName=dyn_cloth_bk_l_04_capsule
   center=<9.3376 0 -0.7694>
   rotation=<rot -89.577 179.9917 79.1496>
  }]
 dyn_cloth_bk_l_05 [{
   shapeType=Capsule
   length=7.9232
   radius=6.8268
   shapeName=dyn_cloth_bk_l_05_capsule
   center=<9.3376 0 -0.7694>
   rotation=<rot -89.5771 179.9926 62.9625>
  }]
 dyn_cloth_bk_r_01 [{
   shapeType=Capsule
   length=9.75
   radius=3.25
   shapeName=dyn_cloth_bk_r_01_capsule
   center=<5.9658 -0.3783 0.9599>
   rotation=<rot -89.2718 -3.6319 0.0006>
  }]
 dyn_cloth_bk_r_02 [{
   shapeType=Capsule
   length=9.75
   radius=3.25
   shapeName=dyn_cloth_bk_r_02_capsule
   center=<5.9658 -0.3783 0.9599>
   rotation=<rot -89.2718 -3.6319 0.0006>
  }]
 dyn_cloth_bk_r_03 [{
   shapeType=Capsule
   length=9.75
   radius=3.25
   shapeName=dyn_cloth_bk_r_03_capsule
   center=<5.9658 -0.3783 0.9599>
   rotation=<rot -89.2718 -3.6319 0.0006>
  }]
 dyn_cloth_fr_l_01 [{
   shapeType=Box
   extentX=16.4799
   extentY=6.5407
   extentZ=10.7628
   shapeName=dyn_cloth_fr_l_01_box
   center=<3.6313 0 0.7509>
   rotation=<rot 0 0 90>
  }]
 dyn_cloth_fr_l_02 [{
   shapeType=Box
   extentX=18.7788
   extentY=6.0952
   extentZ=11.7064
   shapeName=dyn_cloth_fr_l_02_box
   center=<3.6313 0 0.7509>
   rotation=<rot 0 0 90>
  }]
 dyn_cloth_fr_l_03 [{
   shapeType=Box
   extentX=14.1379
   extentY=6.0972
   extentZ=17.9929
   shapeName=dyn_cloth_fr_l_03_box
   center=<3.6313 0 0.7509>
   rotation=<rot 0 0 90>
  }]
 dyn_cloth_fr_l_MSTR [{shapeType=Capsule, length=0.5, radius=0.5}]
 dyn_cloth_fr_r_01 [{
   shapeType=Box
   extentX=12.3844
   extentY=5.9342
   extentZ=9.5214
   shapeName=dyn_cloth_fr_r_01_box
   center=<3.6313 0 0.7509>
   rotation=<rot 0 0 90>
  }]
 dyn_cloth_fr_r_02 [{
   shapeType=Box
   extentX=15.6927
   extentY=5.9846
   extentZ=13.6479
   shapeName=dyn_cloth_fr_r_02_box
   center=<3.6313 0 0.7509>
   rotation=<rot 0 0 90>
  }]
 dyn_cloth_fr_r_03 [{
   shapeType=Box
   extentX=12.5076
   extentY=5.4275
   extentZ=8.8416
   shapeName=dyn_cloth_fr_r_03_box
   center=<3.6313 0 0.7509>
   rotation=<rot 0 0 90>
  }]
 dyn_cloth_fr_r_MSTR [{shapeType=Capsule, length=0.5, radius=0.5}]
 dyn_cloth_l_01 [{shapeType=Box, extentX=10, extentY=14, extentZ=6, shapeName=dyn_cloth_l_01_box, center=<3.3215 -3.0123 1.2009>, rotation=<rot 0 0 -5.2795>}]
 dyn_cloth_l_02 [{
   shapeType=Box
   extentX=13
   extentY=18
   extentZ=6.5
   shapeName=dyn_cloth_l_02_box
   center=<5.9178 -4.3662 -0.1642>
   rotation=<rot 1.7911 -0.4421 -13.8594>
  }]
 dyn_cloth_l_03 [{
   shapeType=Box
   extentX=18
   extentY=18
   extentZ=8.1719
   shapeName=dyn_cloth_l_03_box
   center=<8.3751 -4.4784 0.2908>
   rotation=<rot 0 0 -13.8485>
  }]
 dyn_cloth_l_04 [{
   shapeType=Box
   extentX=24.5
   extentY=27.8991
   extentZ=6.5739
   shapeName=dyn_cloth_l_04_box
   center=<4.3296 -4.868 -1.2758>
   rotation=<rot -0.5756 -2.192 -14.7058>
  }]
 dyn_cloth_l_05 [{
   shapeType=Capsule
   length=11.6917
   radius=8.0288
   shapeName=dyn_cloth_l_05_capsule
   center=<2.5608 -6.364 -1.8989>
   rotation=<rot 0 0 -97.5239>
  }]
 dyn_cloth_l_06 [{
   shapeType=Capsule
   length=6.7989
   radius=8.4216
   shapeName=dyn_cloth_l_06_capsule
   center=<2.8406 -4.0796 -1.5969>
   rotation=<rot 0 0 -97.5239>
  }]
 dyn_cloth_l_MSTR [{shapeType=Capsule, length=0.5, radius=0.5}]
 dyn_cloth_r_01 [{
   shapeType=Capsule
   length=6.0364
   radius=3.2589
   shapeName=dyn_cloth_r_01_capsule
   center=<3.3236 -0.2098 0.8854>
   rotation=<rot -89.2788 -3.6236 0.0005>
  }]
 dyn_cloth_r_02 [{
   shapeType=Box
   extentX=13.2511
   extentY=10.1094
   extentZ=3.6668
   shapeName=dyn_cloth_r_02_box
   center=<5.994 -0.0047 0.1891>
   rotation=<rot -15.0805 0.3647 -2.6188>
  }]
 dyn_cloth_r_03 [{
   shapeType=Box
   extentX=11.1021
   extentY=16.2726
   extentZ=3.4394
   shapeName=dyn_cloth_r_03_box
   center=<4.7587 0 -0.8072>
   rotation=<rot -5.3825 -0.0001 0>
  }]
 dyn_cloth_r_MSTR [{shapeType=Capsule, length=0.5, radius=0.5}]
 dyn_crouch_collision [{shapeType=Capsule, length=0, radius=39.8749, shapeName=dyn_crouch_collision_capsule, center=<-0.0002 -2.3087 -11.4096>}]
 pelvis [{shapeType=Capsule, length=3.7421, radius=10.9145, center=<-10.3754 -0.64 0.0005>}]
 root [{
   shapeType=Capsule
   length=100.25
   radius=59.5
   shapeName=root_capsule
   center=<-0.8461 -29.8621 -54.9708>
   rotation=<rot 89.1188 0.0016 0.0016>
  }]
 thigh_l [
  {shapeType=Capsule, length=30.6836, radius=6, center=<-18.7531 0.0004 0.2337>, rotation=<rot 87.3858 180 180>}
  {
   shapeType=Capsule
   length=8.2361
   radius=7.9917
   shapeName=thigh_l_capsule
   center=<-17.26 0.7722 1.0915>
   rotation=<rot 86.9448 179.999 179.9993>
  }
 ]
 thigh_r [
  {shapeType=Capsule, length=30.1718, radius=6.5118, center=<18.7531 -0.0004 -0.2337>, rotation=<rot 87.3858 180 180>}
  {
   shapeType=Capsule
   length=7.4676
   radius=7.8953
   shapeName=thigh_r_capsule
   center=<18.6714 -0.8162 -0.8522>
   rotation=<rot -83.9657 -0.0011 5.9247>
  }
 ]
}
constraints [
 {bone1Name=dyn_cloth_fr_r_01, bone2Name=dyn_cloth_fr_r_MSTR, swing1LimitDegrees=30, swing2LimitDegrees=30, twistLimitDegrees=5}
 {bone1Name=dyn_cloth_fr_r_02, bone2Name=dyn_cloth_fr_r_01, swing1LimitDegrees=30, swing2LimitDegrees=30, twistLimitDegrees=5}
 {bone1Name=dyn_cloth_fr_r_03, bone2Name=dyn_cloth_fr_r_02, swing1LimitDegrees=45, swing2LimitDegrees=45, twistLimitDegrees=15}
 {bone1Name=dyn_cloth_fr_l_01, bone2Name=dyn_cloth_fr_l_MSTR, swing1LimitDegrees=30, swing2LimitDegrees=30, twistLimitDegrees=5}
 {bone1Name=dyn_cloth_fr_l_02, bone2Name=dyn_cloth_fr_l_01, swing1LimitDegrees=30, swing2LimitDegrees=30, twistLimitDegrees=5}
 {bone1Name=dyn_cloth_fr_l_03, bone2Name=dyn_cloth_fr_l_02, swing1LimitDegrees=45, swing2LimitDegrees=45, twistLimitDegrees=15}
 {bone1Name=dyn_cloth_bk_l_02, bone2Name=dyn_cloth_bk_l_01, swing1LimitDegrees=45, swing2LimitDegrees=45, twistLimitDegrees=15}
 {bone1Name=dyn_cloth_bk_l_03, bone2Name=dyn_cloth_bk_l_02, swing1LimitDegrees=45, swing2LimitDegrees=45, twistLimitDegrees=15}
 {bone1Name=dyn_cloth_bk_l_04, bone2Name=dyn_cloth_bk_l_03, swing1LimitDegrees=45, swing2LimitDegrees=45, twistLimitDegrees=15}
 {bone1Name=dyn_cloth_bk_l_05, bone2Name=dyn_cloth_bk_l_04, swing1LimitDegrees=45, swing2LimitDegrees=45, twistLimitDegrees=15}
 {bone1Name=dyn_cloth_l_02, bone2Name=dyn_cloth_l_01, swing1LimitDegrees=10, swing2LimitDegrees=45, twistLimitDegrees=5}
 {bone1Name=dyn_cloth_l_03, bone2Name=dyn_cloth_l_02, swing1LimitDegrees=45, swing2LimitDegrees=45, twistLimitDegrees=5}
 {bone1Name=dyn_cloth_l_04, bone2Name=dyn_cloth_l_03, swing1LimitDegrees=45, swing2LimitDegrees=45, twistLimitDegrees=15}
 {bone1Name=dyn_cloth_l_05, bone2Name=dyn_cloth_l_04, swing1LimitDegrees=45, swing2LimitDegrees=45, twistLimitDegrees=15}
 {bone1Name=dyn_cloth_l_06, bone2Name=dyn_cloth_l_05, swing1LimitDegrees=45, swing2LimitDegrees=45, twistLimitDegrees=15}
 {bone1Name=dyn_cloth_bk_02, bone2Name=dyn_cloth_bk_01, swing1LimitDegrees=60, swing2LimitDegrees=60, twistLimitDegrees=5}
 {bone1Name=dyn_cloth_bk_03, bone2Name=dyn_cloth_bk_02, swing1LimitDegrees=60, swing2LimitDegrees=60, twistLimitDegrees=5}
 {bone1Name=dyn_cloth_bk_04, bone2Name=dyn_cloth_bk_03, swing1LimitDegrees=60, swing2LimitDegrees=60, twistLimitDegrees=5}
 {bone1Name=dyn_cloth_bk_05, bone2Name=dyn_cloth_bk_04, swing1LimitDegrees=45, swing2LimitDegrees=45, twistLimitDegrees=5}
 {bone1Name=dyn_cloth_bk_r_02, bone2Name=dyn_cloth_bk_r_01, swing1LimitDegrees=60, swing2LimitDegrees=60, twistLimitDegrees=5}
 {bone1Name=dyn_cloth_bk_r_03, bone2Name=dyn_cloth_bk_r_02, swing1LimitDegrees=60, swing2LimitDegrees=60, twistLimitDegrees=5}
 {bone1Name=dyn_cloth_r_02, bone2Name=dyn_cloth_r_01, swing1LimitDegrees=30, swing2LimitDegrees=30, twistLimitDegrees=5}
 {bone1Name=dyn_cloth_r_03, bone2Name=dyn_cloth_r_02, swing1LimitDegrees=30, swing2LimitDegrees=30, twistLimitDegrees=5}
 {
  bone1Name=dyn_cloth_fr_r_MSTR
  bone2Name=pelvis
  swing1LimitDegrees=45
  swing1Motion=Free
  swing2LimitDegrees=45
  swing2Motion=Free
  twistLimitDegrees=45
  twistMotion=Free
 }
 {
  bone1Name=dyn_cloth_fr_l_MSTR
  bone2Name=pelvis
  swing1LimitDegrees=45
  swing1Motion=Free
  swing2LimitDegrees=45
  swing2Motion=Free
  twistLimitDegrees=45
  twistMotion=Free
 }
 {bone1Name=dyn_cloth_bk_l_01, bone2Name=pelvis, swing1LimitDegrees=60, swing2LimitDegrees=60, twistLimitDegrees=5}
 {bone1Name=dyn_cloth_l_01, bone2Name=dyn_cloth_l_MSTR, swing1LimitDegrees=0.1, swing2LimitDegrees=45, twistLimitDegrees=0.1}
 {bone1Name=dyn_cloth_bk_01, bone2Name=pelvis, swing1LimitDegrees=60, swing2LimitDegrees=60, twistLimitDegrees=5}
 {bone1Name=dyn_cloth_bk_r_01, bone2Name=pelvis, swing1LimitDegrees=60, swing2LimitDegrees=60, twistLimitDegrees=5}
 {bone1Name=dyn_cloth_r_01, bone2Name=dyn_cloth_r_MSTR, swing1LimitDegrees=30, swing2LimitDegrees=30, twistLimitDegrees=5}
 {
  bone1Name=thigh_l
  bone2Name=pelvis
  swing1LimitDegrees=45
  swing1Motion=Free
  swing2LimitDegrees=45
  swing2Motion=Free
  twistLimitDegrees=45
  twistMotion=Free
 }
 {
  bone1Name=calf_l
  bone2Name=thigh_l
  swing1LimitDegrees=45
  swing1Motion=Free
  swing2LimitDegrees=45
  swing2Motion=Free
  twistLimitDegrees=45
  twistMotion=Free
 }
 {
  bone1Name=thigh_r
  bone2Name=pelvis
  swing1LimitDegrees=45
  swing1Motion=Free
  swing2LimitDegrees=45
  swing2Motion=Free
  twistLimitDegrees=45
  twistMotion=Free
 }
 {
  bone1Name=calf_r
  bone2Name=thigh_r
  swing1LimitDegrees=45
  swing1Motion=Free
  swing2LimitDegrees=45
  swing2Motion=Free
  twistLimitDegrees=45
  twistMotion=Free
 }
 {
  bone1Name=pelvis
  bone2Name=root
  swing1LimitDegrees=45
  swing1Motion=Free
  swing2LimitDegrees=45
  swing2Motion=Free
  twistLimitDegrees=45
  twistMotion=Free
 }
 {
  bone1Name=dyn_cloth_l_MSTR
  bone2Name=pelvis
  swing1LimitDegrees=45
  swing1Motion=Free
  swing2LimitDegrees=45
  swing2Motion=Free
  twistLimitDegrees=45
  twistMotion=Free
 }
 {
  bone1Name=dyn_cloth_r_MSTR
  bone2Name=pelvis
  swing1LimitDegrees=45
  swing1Motion=Free
  swing2LimitDegrees=45
  swing2Motion=Free
  twistLimitDegrees=45
  twistMotion=Free
 }
 {
  bone1Name=dyn_crouch_collision
  bone2Name=root
  swing1LimitDegrees=45
  swing1Motion=Free
  swing2LimitDegrees=45
  swing2Motion=Free
  twistLimitDegrees=45
  twistMotion=Free
 }
 {
  bone1Name=dyn_cloth_bk_l_05
  bone2Name=dyn_cloth_l_05
  swing1LimitDegrees=45
  swing1Motion=Free
  swing2LimitDegrees=45
  swing2Motion=Free
  twistLimitDegrees=45
  twistMotion=Free
 }
 {
  bone1Name=dyn_cloth_bk_l_05
  bone2Name=dyn_cloth_bk_05
  swing1LimitDegrees=45
  swing1Motion=Free
  swing2LimitDegrees=45
  swing2Motion=Free
  twistLimitDegrees=45
  twistMotion=Free
 }
 {
  bone1Name=dyn_cloth_l_04
  bone2Name=dyn_cloth_bk_l_04
  swing1LimitDegrees=45
  swing1Motion=Free
  swing2LimitDegrees=45
  swing2Motion=Free
  twistLimitDegrees=45
  twistMotion=Free
 }
 {
  bone1Name=dyn_cloth_bk_04
  bone2Name=dyn_cloth_bk_l_04
  swing1LimitDegrees=45
  swing1Motion=Free
  swing2LimitDegrees=45
  swing2Motion=Free
  twistLimitDegrees=45
  twistMotion=Free
 }
 {
  bone1Name=dyn_cloth_bk_l_05
  bone2Name=dyn_cloth_l_06
  swing1LimitDegrees=45
  swing1Motion=Free
  swing2LimitDegrees=45
  swing2Motion=Free
  twistLimitDegrees=45
  twistMotion=Free
 }
 {
  bone1Name=dyn_cloth_fr_r_02
  bone2Name=dyn_cloth_fr_l_02
  swing1LimitDegrees=45
  swing1Motion=Free
  swing2LimitDegrees=45
  swing2Motion=Free
  twistLimitDegrees=45
  twistMotion=Free
 }
 {
  bone1Name=dyn_cloth_fr_r_03
  bone2Name=dyn_cloth_fr_l_03
  swing1LimitDegrees=45
  swing1Motion=Free
  swing2LimitDegrees=45
  swing2Motion=Free
  twistLimitDegrees=45
  twistMotion=Free
 }
]
