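# Physics asset for the dynamic torso mesh of the male medium Golden Skeleton cosmetic body, covering chest cloth and a necklace.
# Capsules on the spine, neck and clavicles form the torso volume, linked by constraints with free swing and twist.
# Three three-bone chains (left cloth, right cloth, necklace) hang from spine_05 with tight twist limits.
# Cloth chains are nearly rigid at the root and looser below, while necklace swing limits are asymmetric.
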PhysicsAsset /BRCosmetics/Characters/Player/Male/Medium/Bodies/M_MED_The_Golden_Skeleton/Meshes/M_MED_The_Golden_Skeleton_Dyn_torso1
bodies {
 clavicle_l [{shapeType=Capsule, length=9.0126, radius=7.1972, center=<10.2698 -0.0003 1.1248>, rotation=<rot 85.3739 0 0>}]
 clavicle_r [{shapeType=Capsule, length=9.0126, radius=7.1972, center=<-10.2698 0.0003 -1.1248>, rotation=<rot 85.3739 0 0>}]
 dyn_chest_cloth_l_01 [{
   shapeType=Capsule
   length=3.3281
   radius=3.1519
   shapeName=dyn_chest_cloth_l_01_capsule
   center=<3.1123 0.3793 0.2613>
   rotation=<rot 0 -41.3205 -86.2937>
  }]
 dyn_chest_cloth_l_02 [{
   shapeType=Capsule
   length=2.9797
   radius=2.0173
   shapeName=dyn_chest_cloth_l_02_capsule
   center=<2.0004 -0.1634 0.182>
   rotation=<rot -27.431 -92.4346 -103.9376>
  }]
 dyn_chest_cloth_l_03 [{
   shapeType=Capsule
   length=6.0783
   radius=2.0585
   shapeName=dyn_chest_cloth_l_03_capsule
   center=<2.1311 -0.2604 -0.2255>
   rotation=<rot 0 -95.534 -87.0051>
  }]
 dyn_chest_cloth_r_01 [{
   shapeType=Capsule
   length=3.4159
   radius=2.7669
   shapeName=dyn_chest_cloth_r_01_capsule
   center=<3.8354 0.2943 0.1205>
   rotation=<rot 0 159.7935 -84.2248>
  }]
 dyn_chest_cloth_r_02 [{
   shapeType=Capsule
   length=3.717
   radius=1.418
   shapeName=dyn_chest_cloth_r_02_capsule
   center=<3.547 1.0211 0.4635>
   rotation=<rot 0 -125.2534 -88.7499>
  }]
 dyn_chest_cloth_r_03 [{
   shapeType=Capsule
   length=2.992
   radius=1.0467
   shapeName=dyn_chest_cloth_r_03_capsule
   center=<3.3128 -0.1285 0.0484>
   rotation=<rot 0 -101.584 -83.6265>
  }]
 dyn_necklace_01 [{
   shapeType=Capsule
   length=8.75
   radius=2.25
   shapeName=dyn_necklace_01_capsule
   center=<3.9169 0.0001 1.1587>
   rotation=<rot 0 -89.9999 85.4183>
  }]
 dyn_necklace_02 [{
   shapeType=Capsule
   length=4.755
   radius=2.505
   shapeName=dyn_necklace_02_capsule
   center=<2.5956 0 0.7838>
   rotation=<rot 0 -89.9999 93.8686>
  }]
 dyn_necklace_03 [{
   shapeType=Capsule
   length=7.7804
   radius=1.2692
   shapeName=dyn_necklace_03_capsule
   center=<3.3024 0.19 -0.1357>
   rotation=<rot -1.047 -88.8929 -87.4756>
  }]
 neck_01 [{shapeType=Capsule, length=13.1149, radius=4.7564, center=<9.0898 -0.9485 0.0001>, rotation=<rot 90 180 -175.6909>}]
 spine_03 [{shapeType=Capsule, length=3.1252, radius=11.3124, center=<2.2877 1.8564 0>}]
 spine_04 [{shapeType=Capsule, length=7.777, radius=11.6919, center=<3.0336 0.6597 0>}]
 spine_05 [
  {shapeType=Capsule, length=16.2708, radius=9.1, center=<3.8549 -1.2455 0>}
  {shapeType=Capsule, length=16.4265, radius=5.6, center=<7.018 -3.7593 0.0001>}
  {shapeType=Capsule, length=15.654, radius=9.1, center=<-1.1877 1.9023 0>}
  {shapeType=Capsule, length=15, radius=3.75, shapeName=spine_05_capsule, center=<10.5157 -8.9562 0>}
 ]
}
constraints [
 {
  bone1Name=spine_05
  bone2Name=spine_04
  swing1LimitDegrees=45
  swing1Motion=Free
  swing2LimitDegrees=45
  swing2Motion=Free
  twistLimitDegrees=45
  twistMotion=Free
 }
 {
  bone1Name=spine_04
  bone2Name=spine_03
  swing1LimitDegrees=45
  swing1Motion=Free
  swing2LimitDegrees=45
  swing2Motion=Free
  twistLimitDegrees=45
  twistMotion=Free
 }
 {
  bone1Name=clavicle_l
  bone2Name=spine_05
  swing1LimitDegrees=45
  swing1Motion=Free
  swing2LimitDegrees=45
  swing2Motion=Free
  twistLimitDegrees=45
  twistMotion=Free
 }
 {
  bone1Name=clavicle_r
  bone2Name=spine_05
  swing1LimitDegrees=45
  swing1Motion=Free
  swing2LimitDegrees=45
  swing2Motion=Free
  twistLimitDegrees=45
  twistMotion=Free
 }
 {
  bone1Name=neck_01
  bone2Name=spine_05
  swing1LimitDegrees=45
  swing1Motion=Free
  swing2LimitDegrees=45
  swing2Motion=Free
  twistLimitDegrees=45
  twistMotion=Free
 }
 {bone1Name=dyn_chest_cloth_l_01, bone2Name=spine_05, swing1LimitDegrees=1, swing2LimitDegrees=1, twistLimitDegrees=1}
 {bone1Name=dyn_chest_cloth_l_02, bone2Name=dyn_chest_cloth_l_01, swing1LimitDegrees=15, swing2LimitDegrees=15, twistLimitDegrees=1}
 {bone1Name=dyn_chest_cloth_l_03, bone2Name=dyn_chest_cloth_l_02, swing1LimitDegrees=15, swing2LimitDegrees=15, twistLimitDegrees=1}
 {bone1Name=dyn_chest_cloth_r_01, bone2Name=spine_05, swing1LimitDegrees=1, swing2LimitDegrees=1, twistLimitDegrees=1}
 {bone1Name=dyn_chest_cloth_r_02, bone2Name=dyn_chest_cloth_r_01, swing1LimitDegrees=15, swing2LimitDegrees=15, twistLimitDegrees=1}
 {bone1Name=dyn_chest_cloth_r_03, bone2Name=dyn_chest_cloth_r_02, swing1LimitDegrees=15, swing2LimitDegrees=15, twistLimitDegrees=1}
 {bone1Name=dyn_necklace_01, bone2Name=spine_05, swing1LimitDegrees=10, swing2LimitDegrees=20, twistLimitDegrees=1}
 {bone1Name=dyn_necklace_02, bone2Name=dyn_necklace_01, swing1LimitDegrees=10, swing2LimitDegrees=20, twistLimitDegrees=1}
 {bone1Name=dyn_necklace_03, bone2Name=dyn_necklace_02, swing1LimitDegrees=10, swing2LimitDegrees=20, twistLimitDegrees=1}
]
