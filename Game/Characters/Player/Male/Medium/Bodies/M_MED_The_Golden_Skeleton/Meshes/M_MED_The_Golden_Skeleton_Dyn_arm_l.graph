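# Physics asset for dynamic clothing bones on the left arm of the medium male Golden Skeleton player body.
# Larger capsules cover the shoulder, three bicep sleeve bones and forearm cuff, while tiny capsules on the master shoulder, upper-arm twist, lower arm and hand bones serve as anchors.
# Shoulder and sleeve joints have tight twist with one wider swing axis, the cuff is tight on all axes, and the hand-to-lowerarm joint rotates freely.
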PhysicsAsset /Game/Characters/Player/Male/Medium/Bodies/M_MED_The_Golden_Skeleton/Meshes/M_MED_The_Golden_Skeleton_Dyn_arm_l
bodies {
 dyn_cuff_forearm_l [{shapeType=Capsule, length=2.5, radius=2.2, center=<0.5804 -0.0001 0>, rotation=<rot 90 0 0>}]
 dyn_shoulder_l [{shapeType=Capsule, length=2.5, radius=2.2, center=<1 0 0>, rotation=<rot 90 0 0>}]
 dyn_shoulder_l_MSTR [{shapeType=Capsule, length=0.5, radius=0.5}]
 dyn_sleeve_bicep_l_01 [{shapeType=Capsule, length=2.5, radius=2.2, center=<1 0 0>, rotation=<rot 90 0 0>}]
 dyn_sleeve_bicep_l_02 [{shapeType=Capsule, length=2.5, radius=2.2, center=<1 0 0>, rotation=<rot 90 0 0>}]
 dyn_sleeve_bicep_l_03 [{shapeType=Capsule, length=2.5, radius=2.2, center=<1 0 0>, rotation=<rot 90 0 0>}]
 hand_ulnarad_l [{shapeType=Capsule, length=0.5, radius=0.5}]
 lowerarm_l [{shapeType=Capsule, length=0.5, radius=0.5}]
 upperarm_twist_01_l [{shapeType=Capsule, length=0.5, radius=0.5}]
}
constraints [
 {bone1Name=dyn_shoulder_l, bone2Name=dyn_shoulder_l_MSTR, swing1LimitDegrees=5, swing2LimitDegrees=45, twistLimitDegrees=5}
 {bone1Name=dyn_cuff_forearm_l, bone2Name=hand_ulnarad_l, swing1LimitDegrees=5, swing2LimitDegrees=5, twistLimitDegrees=5}
 {bone1Name=dyn_sleeve_bicep_l_03, bone2Name=upperarm_twist_01_l, swing1LimitDegrees=5, swing2LimitDegrees=45, twistLimitDegrees=5}
 {bone1Name=dyn_sleeve_bicep_l_02, bone2Name=upperarm_twist_01_l, swing1LimitDegrees=5, swing2LimitDegrees=45, twistLimitDegrees=5}
 {bone1Name=dyn_sleeve_bicep_l_01, bone2Name=upperarm_twist_01_l, swing1LimitDegrees=5, swing2LimitDegrees=45, twistLimitDegrees=5}
 {
  bone1Name=hand_ulnarad_l
  bone2Name=lowerarm_l
  swing1LimitDegrees=45
  swing1Motion=Free
  swing2LimitDegrees=45
  swing2Motion=Free
  twistLimitDegrees=45
  twistMotion=Free
 }
]
