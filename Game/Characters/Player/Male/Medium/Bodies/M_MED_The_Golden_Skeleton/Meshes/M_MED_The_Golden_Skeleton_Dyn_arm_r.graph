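# Physics asset for the right-arm dynamic bones of the male medium Golden Skeleton player body, covering a shoulder piece, three bicep sleeve bones and a bracelet.
# Capsules wrap each bone, with small ones on the arm and shoulder-master anchors.
# Shoulder and sleeve bones get narrow twist and asymmetric swing limits against their anchors.
# The bracelet twists freely with limited swing on hand_ulnarad_r, carrying angularly locked side and up-down bones, while hand_ulnarad_r rotates freely on the forearm.
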PhysicsAsset /Game/Characters/Player/Male/Medium/Bodies/M_MED_The_Golden_Skeleton/Meshes/M_MED_The_Golden_Skeleton_Dyn_arm_r
bodies {
 dyn_bracelet_r [{shapeType=Capsule, length=2.5, radius=2.2, center=<3.7807 -2.8051 -2.4386>}]
 dyn_bracelet_r_side_movement [{shapeType=Capsule, length=2.5, radius=2.2, center=<0 2 0.6986>}]
 dyn_bracelet_r_updown_movement [{shapeType=Capsule, length=2.5, radius=2.2, center=<0 2 0.6986>}]
 dyn_shoulder_r [{shapeType=Capsule, length=2.5, radius=2.2, center=<1 0 0>, rotation=<rot 90 0 0>}]
 dyn_shoulder_r_MSTR [{shapeType=Capsule, length=0.5, radius=0.5}]
 dyn_sleeve_bicep_r_01 [{shapeType=Capsule, length=2.5, radius=2.2, center=<1 0 0>, rotation=<rot 90 0 0>}]
 dyn_sleeve_bicep_r_02 [{shapeType=Capsule, length=2.5, radius=2.2, center=<1 0 0>, rotation=<rot 90 0 0>}]
 dyn_sleeve_bicep_r_03 [{shapeType=Capsule, length=2.5, radius=2.2, center=<1 0 0>, rotation=<rot 90 0 0>}]
 hand_ulnarad_r [{shapeType=Capsule, length=0.5, radius=0.5}]
 lowerarm_r [{shapeType=Capsule, length=0.5, radius=0.5}]
 upperarm_twist_01_r [{shapeType=Capsule, length=0.5, radius=0.5}]
}
constraints [
 {bone1Name=dyn_shoulder_r, bone2Name=dyn_shoulder_r_MSTR, swing1LimitDegrees=5, swing2LimitDegrees=45, twistLimitDegrees=5}
 {bone1Name=dyn_bracelet_r, bone2Name=hand_ulnarad_r, swing1LimitDegrees=10, swing2LimitDegrees=10, twistLimitDegrees=30, twistMotion=Free}
 {bone1Name=dyn_sleeve_bicep_r_01, bone2Name=upperarm_twist_01_r, swing1LimitDegrees=5, swing2LimitDegrees=45, twistLimitDegrees=5}
 {bone1Name=dyn_sleeve_bicep_r_02, bone2Name=upperarm_twist_01_r, swing1LimitDegrees=5, swing2LimitDegrees=45, twistLimitDegrees=5}
 {bone1Name=dyn_sleeve_bicep_r_03, bone2Name=upperarm_twist_01_r, swing1LimitDegrees=5, swing2LimitDegrees=45, twistLimitDegrees=5}
 {
  bone1Name=dyn_bracelet_r_side_movement
  bone2Name=dyn_bracelet_r
  swing1LimitDegrees=10
  swing1Motion=Locked
  swing2LimitDegrees=10
  swing2Motion=Locked
  twistLimitDegrees=30
  twistMotion=Locked
 }
 {
  bone1Name=dyn_bracelet_r_updown_movement
  bone2Name=dyn_bracelet_r_side_movement
  swing1LimitDegrees=10
  swing1Motion=Locked
  swing2LimitDegrees=10
  swing2Motion=Locked
  twistLimitDegrees=30
  twistMotion=Locked
 }
 {
  bone1Name=hand_ulnarad_r
  bone2Name=lowerarm_r
  swing1LimitDegrees=45
  swing1Motion=Free
  swing2LimitDegrees=45
  swing2Motion=Free
  twistLimitDegrees=45
  twistMotion=Free
 }
]
