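# Physics asset for the right leg of the Golden Skeleton medium male player body, set up for dynamic shoelace motion.
# Capsule bodies cover the calf, a small shoelace master bone and two two-bone lace chains, inner and outer.
# Both chains hang from the master through constraints with narrow swing and tighter twist limit angles.
# The master joins the calf through a constraint with free swing and twist.
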PhysicsAsset /Game/Characters/Player/Male/Medium/Bodies/M_MED_The_Golden_Skeleton/Meshes/M_MED_The_Golden_Skeleton_Dyn_leg_r
bodies {
 calf_r [{shapeType=Capsule, length=22.4767, radius=3.5233, center=<28.7477 1.5793 0.2853>, rotation=<rot 87.8141 -10.7107 0.0001>}]
 dyn_shoelace_in_r_01 [{shapeType=Capsule, length=3.2596, radius=1.4404, center=<2 0 0>, rotation=<rot 90 0 0>}]
 dyn_shoelace_in_r_02 [{shapeType=Capsule, length=3.2596, radius=1.4404, center=<2 0 0>, rotation=<rot 90 0 0>}]
 dyn_shoelace_out_r_01 [{shapeType=Capsule, length=3.2596, radius=1.4404, center=<2 0 0>, rotation=<rot 90 0 0>}]
 dyn_shoelace_out_r_02 [{shapeType=Capsule, length=3.2596, radius=1.4404, center=<2 0 0>, rotation=<rot 90 0 0>}]
 dyn_shoelace_r_MSTR [{shapeType=Capsule, length=0.5, radius=0.5}]
}
constraints [
 {bone1Name=dyn_shoelace_out_r_01, bone2Name=dyn_shoelace_r_MSTR, swing1LimitDegrees=20, swing2LimitDegrees=20, twistLimitDegrees=5}
 {bone1Name=dyn_shoelace_out_r_02, bone2Name=dyn_shoelace_out_r_01, swing1LimitDegrees=20, swing2LimitDegrees=20, twistLimitDegrees=5}
 {bone1Name=dyn_shoelace_in_r_01, bone2Name=dyn_shoelace_r_MSTR, swing1LimitDegrees=20, swing2LimitDegrees=20, twistLimitDegrees=5}
 {bone1Name=dyn_shoelace_in_r_02, bone2Name=dyn_shoelace_in_r_01, swing1LimitDegrees=20, swing2LimitDegrees=20, twistLimitDegrees=5}
 {
  bone1Name=dyn_shoelace_r_MSTR
  bone2Name=calf_r
  swing1LimitDegrees=45
  swing1Motion=Free
  swing2LimitDegrees=45
  swing2Motion=Free
  twistLimitDegrees=45
  twistMotion=Free
 }
]
