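# Physics asset for The Golden Skeleton male player body's dynamic skirt, named as the skydiving variant.
# Box-body bone chains hang from the pelvis: front and side chains through tiny MSTR capsules, back chains directly.
# Limited constraints make back and side chains mostly single-axis swingers, while front chains swing on two axes.
# Free-constrained leg and pelvis shapes, a wide flat root box and a large crouch sphere supply collision volumes.
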PhysicsAsset /Game/Characters/Player/Male/Medium/Bodies/M_MED_The_Golden_Skeleton/Meshes/M_MED_The_Golden_Skeleton_Dyn_skirt_skydiving
bodies {
 calf_l [
  {shapeType=Sphere, radius=12.0164, center=<-10.5679 -3.1541 -0.3629>}
  {shapeType=Capsule, length=29.7546, radius=6, center=<-25.7076 -1.456 -0.0001>, rotation=<rot 90 172.875 -178.2732>}
 ]
 calf_r [
  {shapeType=Sphere, radius=9.2664, center=<13.2948 3.1541 0.363>}
  {shapeType=Capsule, length=29.7546, radius=6, center=<25.7076 1.456 0.0001>, rotation=<rot 90 172.875 -178.2732>}
 ]
 dyn_cloth_bk_01 [{shapeType=Box, extentX=12.1318, extentY=10.0287, extentZ=1.622, center=<6.3 0 -0.0002>}]
 dyn_cloth_bk_02 [{shapeType=Box, extentX=12.1318, extentY=10.0287, extentZ=1.622, center=<6.3 0 -0.0002>}]
 dyn_cloth_bk_03 [{shapeType=Box, extentX=12.1318, extentY=14.5686, extentZ=1.622, center=<6.3 0 -0.0002>}]
 dyn_cloth_bk_04 [{shapeType=Box, extentX=12.1318, extentY=17.3125, extentZ=1.622, center=<6.3 0 -0.0002>}]
 dyn_cloth_bk_05 [{shapeType=Box, extentX=12.1318, extentY=17.3125, extentZ=1.622, center=<6.3 5 0>}]
 dyn_cloth_bk_l_01 [{shapeType=Box, extentX=12.1318, extentY=10.0287, extentZ=1.622, center=<6.3 0 -0.0002>}]
 dyn_cloth_bk_l_02 [{shapeType=Box, extentX=12.1318, extentY=10.0287, extentZ=1.622, center=<6.3 0 -0.0002>}]
 dyn_cloth_bk_l_03 [{shapeType=Box, extentX=12.1318, extentY=18.0344, extentZ=1.622, center=<6.3 0 -0.0002>}]
 dyn_cloth_bk_l_04 [{shapeType=Box, extentX=12.1318, extentY=18.0344, extentZ=1.622, center=<6.3 0 -0.0002>}]
 dyn_cloth_bk_l_05 [{shapeType=Box, extentX=12.1318, extentY=18.0344, extentZ=1.622, center=<6.3 0 -0.0002>}]
 dyn_cloth_bk_r_01 [{shapeType=Box, extentX=9.5052, extentY=10.0287, extentZ=1.622, center=<4.7489 0 -0.0002>}]
 dyn_cloth_bk_r_02 [{shapeType=Box, extentX=12.2285, extentY=10.8367, extentZ=1.2052, center=<6.2789 0 -0.0005>}]
 dyn_cloth_bk_r_03 [{shapeType=Box, extentX=12.1113, extentY=10.8367, extentZ=1.2052, center=<5.9721 0.0001 -0.0002>}]
 dyn_cloth_fr_l_01 [{shapeType=Box, extentX=12.1318, extentY=4.0482, extentZ=1.622, center=<6.3 -0.5 0>}]
 dyn_cloth_fr_l_02 [{shapeType=Box, extentX=12.1318, extentY=4.0482, extentZ=1.622, center=<6.3 -0.5 0>}]
 dyn_cloth_fr_l_03 [{shapeType=Box, extentX=12.1318, extentY=4.0482, extentZ=1.622, center=<6.3 -0.5 0>}]
 dyn_cloth_fr_l_MSTR [{shapeType=Capsule, length=0.5, radius=0.5}]
 dyn_cloth_fr_r_01 [{shapeType=Box, extentX=12.1318, extentY=4.0482, extentZ=1.622, center=<6.3266 0.5 0>}]
 dyn_cloth_fr_r_02 [{shapeType=Box, extentX=12.1318, extentY=4.0482, extentZ=1.622, center=<6.3074 0.5 0>}]
 dyn_cloth_fr_r_03 [{shapeType=Box, extentX=12.1318, extentY=4.0482, extentZ=1.622, center=<6.3058 0.5 0>}]
 dyn_cloth_fr_r_MSTR [{shapeType=Capsule, length=0.5, radius=0.5}]
 dyn_cloth_l_01 [{shapeType=Box, extentX=12.1318, extentY=10.0287, extentZ=1.622, center=<6.3 0 -0.0002>, rotation=<rot 0 0 14.6763>}]
 dyn_cloth_l_02 [{shapeType=Box, extentX=12.1318, extentY=10.0287, extentZ=1.622, center=<6.3 0 -0.0002>}]
 dyn_cloth_l_03 [{shapeType=Box, extentX=12.1318, extentY=20.2684, extentZ=1.622, center=<6.3 -4.5 0>}]
 dyn_cloth_l_04 [{shapeType=Box, extentX=12.1318, extentY=20.2684, extentZ=1.622, center=<6.3 -4.5 0>}]
 dyn_cloth_l_05 [{shapeType=Box, extentX=12.1318, extentY=20.2684, extentZ=1.622, center=<6.3 -4.5 0>}]
 dyn_cloth_l_06 [{shapeType=Box, extentX=12.1318, extentY=20.2684, extentZ=1.622, center=<6.3 -4.5 0>}]
 dyn_cloth_l_MSTR [{shapeType=Capsule, length=0.5, radius=0.5}]
 dyn_cloth_r_01 [{shapeType=Box, extentX=7.7611, extentY=10.7457, extentZ=1.2052, center=<3.9622 0 -0.0002>}]
 dyn_cloth_r_02 [{shapeType=Box, extentX=11.6965, extentY=12.6494, extentZ=2.237, center=<5.7985 0 -0.0004>}]
 dyn_cloth_r_03 [{shapeType=Box, extentX=7.676, extentY=16.1779, extentZ=2.237, center=<4.0248 0 -0.0001>}]
 dyn_cloth_r_MSTR [{shapeType=Capsule, length=0.5, radius=0.5}]
 dyn_crouch_collision [{shapeType=Sphere, radius=41.7434, center=<-0.4526 -1.8161 -10.7878>}]
 pelvis [
  {shapeType=Capsule, length=5.0975, radius=9.5591, center=<-5.788 -1.12 0.0005>}
  {shapeType=Capsule, length=24, radius=6, shapeName=pelvis_capsule, center=<-27.0967 0 -0.5273>, rotation=<rot 90 0.0198 0.0199>}
 ]
 root [
  {shapeType=Capsule, length=0.505, radius=0.505, rotation=<rot 0 -90 0>}
  {shapeType=Box, extentX=172.3247, extentY=172.3247, extentZ=10.5188, center=<0 -0.0001 -3.3177>}
 ]
 thigh_l [
  {shapeType=Capsule, length=30.6836, radius=6, center=<-18.7531 0.0004 0.2337>, rotation=<rot 87.3858 180 180>}
  {shapeType=Capsule, length=11.777, radius=5.5695, center=<-15.4756 3.1533 -1.6642>, rotation=<rot 49.25 99.1701 103.3827>}
 ]
 thigh_r [
  {shapeType=Capsule, length=29.8007, radius=6.8829, center=<18.7531 -0.0004 -0.2337>, rotation=<rot 87.3858 180 180>}
  {shapeType=Capsule, length=11.777, radius=5.5695, center=<15.4756 -3.1533 1.6642>, rotation=<rot 49.25 99.1701 103.3827>}
 ]
}
constraints [
 {bone1Name=dyn_cloth_fr_r_01, bone2Name=dyn_cloth_fr_r_MSTR, swing1LimitDegrees=60, swing2LimitDegrees=60, twistLimitDegrees=5}
 {bone1Name=dyn_cloth_fr_r_02, bone2Name=dyn_cloth_fr_r_01, swing1LimitDegrees=60, swing2LimitDegrees=60, twistLimitDegrees=5}
 {bone1Name=dyn_cloth_fr_r_03, bone2Name=dyn_cloth_fr_r_02, swing1LimitDegrees=60, swing2LimitDegrees=60, twistLimitDegrees=5}
 {bone1Name=dyn_cloth_fr_l_01, bone2Name=dyn_cloth_fr_l_MSTR, swing1LimitDegrees=60, swing2LimitDegrees=60, twistLimitDegrees=5}
 {bone1Name=dyn_cloth_fr_l_02, bone2Name=dyn_cloth_fr_l_01, swing1LimitDegrees=60, swing2LimitDegrees=60, twistLimitDegrees=5}
 {bone1Name=dyn_cloth_fr_l_03, bone2Name=dyn_cloth_fr_l_02, swing1LimitDegrees=60, swing2LimitDegrees=60, twistLimitDegrees=5}
 {bone1Name=dyn_cloth_bk_l_02, bone2Name=dyn_cloth_bk_l_01, swing1LimitDegrees=5, swing2LimitDegrees=60, twistLimitDegrees=5}
 {bone1Name=dyn_cloth_bk_l_03, bone2Name=dyn_cloth_bk_l_02, swing1LimitDegrees=5, swing2LimitDegrees=60, twistLimitDegrees=5}
 {bone1Name=dyn_cloth_bk_l_04, bone2Name=dyn_cloth_bk_l_03, swing1LimitDegrees=5, swing2LimitDegrees=90, twistLimitDegrees=15}
 {bone1Name=dyn_cloth_bk_l_05, bone2Name=dyn_cloth_bk_l_04, swing1LimitDegrees=5, swing2LimitDegrees=90, twistLimitDegrees=15}
 {bone1Name=dyn_cloth_l_02, bone2Name=dyn_cloth_l_01, swing1LimitDegrees=5, swing2LimitDegrees=60, twistLimitDegrees=5}
 {bone1Name=dyn_cloth_l_03, bone2Name=dyn_cloth_l_02, swing1LimitDegrees=5, swing2LimitDegrees=60, twistLimitDegrees=5}
 {bone1Name=dyn_cloth_l_04, bone2Name=dyn_cloth_l_03, swing1LimitDegrees=5, swing2LimitDegrees=90, twistLimitDegrees=15}
 {bone1Name=dyn_cloth_l_05, bone2Name=dyn_cloth_l_04, swing1LimitDegrees=5, swing2LimitDegrees=90, twistLimitDegrees=15}
 {bone1Name=dyn_cloth_l_06, bone2Name=dyn_cloth_l_05, swing1LimitDegrees=5, swing2LimitDegrees=90, twistLimitDegrees=15}
 {bone1Name=dyn_cloth_bk_02, bone2Name=dyn_cloth_bk_01, swing1LimitDegrees=5, swing2LimitDegrees=60, twistLimitDegrees=5}
 {bone1Name=dyn_cloth_bk_03, bone2Name=dyn_cloth_bk_02, swing1LimitDegrees=5, swing2LimitDegrees=60, twistLimitDegrees=5}
 {bone1Name=dyn_cloth_bk_04, bone2Name=dyn_cloth_bk_03, swing1LimitDegrees=5, swing2LimitDegrees=60, twistLimitDegrees=5}
 {bone1Name=dyn_cloth_bk_05, bone2Name=dyn_cloth_bk_04, swing1LimitDegrees=5, swing2LimitDegrees=60, twistLimitDegrees=5}
 {bone1Name=dyn_cloth_bk_r_02, bone2Name=dyn_cloth_bk_r_01, swing1LimitDegrees=5, swing2LimitDegrees=60, twistLimitDegrees=5}
 {bone1Name=dyn_cloth_bk_r_03, bone2Name=dyn_cloth_bk_r_02, swing1LimitDegrees=5, swing2LimitDegrees=60, twistLimitDegrees=5}
 {bone1Name=dyn_cloth_r_02, bone2Name=dyn_cloth_r_01, swing1LimitDegrees=5, swing2LimitDegrees=60, twistLimitDegrees=5}
 {bone1Name=dyn_cloth_r_03, bone2Name=dyn_cloth_r_02, swing1LimitDegrees=5, swing2LimitDegrees=60, twistLimitDegrees=5}
 {
  bone1Name=dyn_cloth_fr_r_MSTR
  bone2Name=pelvis
  swing1LimitDegrees=45
  swing1Motion=Free
  swing2LimitDegrees=45
  swing2Motion=Free
  twistLimitDegrees=45
  twistMotion=Free
 }
 {
  bone1Name=dyn_cloth_fr_l_MSTR
  bone2Name=pelvis
  swing1LimitDegrees=45
  swing1Motion=Free
  swing2LimitDegrees=45
  swing2Motion=Free
  twistLimitDegrees=45
  twistMotion=Free
 }
 {bone1Name=dyn_cloth_bk_l_01, bone2Name=pelvis, swing1LimitDegrees=5, swing2LimitDegrees=60, twistLimitDegrees=5}
 {bone1Name=dyn_cloth_l_01, bone2Name=dyn_cloth_l_MSTR, swing1LimitDegrees=5, swing2LimitDegrees=60, twistLimitDegrees=5}
 {bone1Name=dyn_cloth_bk_01, bone2Name=pelvis, swing1LimitDegrees=5, swing2LimitDegrees=60, twistLimitDegrees=5}
 {bone1Name=dyn_cloth_bk_r_01, bone2Name=pelvis, swing1LimitDegrees=5, swing2LimitDegrees=60, twistLimitDegrees=5}
 {bone1Name=dyn_cloth_r_01, bone2Name=dyn_cloth_r_MSTR, swing1LimitDegrees=5, swing2LimitDegrees=60, twistLimitDegrees=5}
 {
  bone1Name=thigh_l
  bone2Name=pelvis
  swing1LimitDegrees=45
  swing1Motion=Free
  swing2LimitDegrees=45
  swing2Motion=Free
  twistLimitDegrees=45
  twistMotion=Free
 }
 {
  bone1Name=calf_l
  bone2Name=thigh_l
  swing1LimitDegrees=45
  swing1Motion=Free
  swing2LimitDegrees=45
  swing2Motion=Free
  twistLimitDegrees=45
  twistMotion=Free
 }
 {
  bone1Name=thigh_r
  bone2Name=pelvis
  swing1LimitDegrees=45
  swing1Motion=Free
  swing2LimitDegrees=45
  swing2Motion=Free
  twistLimitDegrees=45
  twistMotion=Free
 }
 {
  bone1Name=calf_r
  bone2Name=thigh_r
  swing1LimitDegrees=45
  swing1Motion=Free
  swing2LimitDegrees=45
  swing2Motion=Free
  twistLimitDegrees=45
  twistMotion=Free
 }
 {
  bone1Name=pelvis
  bone2Name=root
  swing1LimitDegrees=45
  swing1Motion=Free
  swing2LimitDegrees=45
  swing2Motion=Free
  twistLimitDegrees=45
  twistMotion=Free
 }
 {
  bone1Name=dyn_cloth_l_MSTR
  bone2Name=pelvis
  swing1LimitDegrees=45
  swing1Motion=Free
  swing2LimitDegrees=45
  swing2Motion=Free
  twistLimitDegrees=45
  twistMotion=Free
 }
 {
  bone1Name=dyn_cloth_r_MSTR
  bone2Name=pelvis
  swing1LimitDegrees=45
  swing1Motion=Free
  swing2LimitDegrees=45
  swing2Motion=Free
  twistLimitDegrees=45
  twistMotion=Free
 }
 {
  bone1Name=dyn_crouch_collision
  bone2Name=root
  swing1LimitDegrees=45
  swing1Motion=Free
  swing2LimitDegrees=45
  swing2Motion=Free
  twistLimitDegrees=45
  twistMotion=Free
 }
]
