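# Physics asset for the Golden Skeleton player torso's dynamic accessories, named as the skydiving variant.
# Spine, clavicle and neck shapes form the body, with chest cloth, front necklace and both necklace pin masters constrained to spine_05.
# Side and back necklace chains close into loops through free-angular constraints at the pins and center pendant chain.
# High and low belt chains hang from their own master bodies with tight limits between segments, unconstrained to the spine.
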PhysicsAsset /Game/Characters/Player/Male/Medium/Bodies/M_MED_The_Golden_Skeleton/Meshes/M_MED_The_Golden_Skeleton_Dyn_torso_skydiving
bodies {
 clavicle_l [{shapeType=Capsule, length=9.0126, radius=7.1972, center=<10.2698 -0.0003 1.1248>, rotation=<rot 85.3739 0 0>}]
 clavicle_r [{shapeType=Capsule, length=9.0126, radius=7.1972, center=<-10.2698 0.0003 -1.1248>, rotation=<rot 85.3739 0 0>}]
 dyn_belt_high_01 [{shapeType=Capsule, length=0.5, radius=0.5}]
 dyn_belt_high_02 [{shapeType=Box, extentX=6.67, extentY=4.9242, extentZ=1.2052, center=<3.6069 0 -0.0002>}]
 dyn_belt_high_03 [{shapeType=Box, extentX=6.67, extentY=4.9242, extentZ=1.2052, center=<3.6069 0 -0.0002>}]
 dyn_belt_high_MSTR [{shapeType=Capsule, length=0.5, radius=0.5}]
 dyn_belt_low_01 [{shapeType=Capsule, length=0.5, radius=0.5}]
 dyn_belt_low_02 [{shapeType=Box, extentX=6.67, extentY=4.9242, extentZ=1.2052, center=<3.6069 0 -0.0002>}]
 dyn_belt_low_03 [{shapeType=Box, extentX=6.67, extentY=4.9242, extentZ=1.2052, center=<3.6069 0 -0.0002>}]
 dyn_belt_low_MSTR [{shapeType=Capsule, length=0.5, radius=0.5}]
 dyn_chest_cloth_l_01 [
  {shapeType=Capsule, length=5.6371, radius=0.7611, center=<3.5239 1.6 0>, rotation=<rot 90 0 0>}
  {shapeType=Capsule, length=5.6371, radius=0.7611, center=<3.5343 -1.2889 0>, rotation=<rot 90 0 0>}
 ]
 dyn_chest_cloth_l_02 [{shapeType=Box, extentX=6.67, extentY=3.5227, extentZ=1.2052, center=<3.6069 0 0>}]
 dyn_chest_cloth_l_03 [{shapeType=Box, extentX=6.67, extentY=1.9739, extentZ=1.2052, center=<3.6069 0 0>}]
 dyn_chest_cloth_r_01 [
  {shapeType=Capsule, length=4.9607, radius=0.7611, center=<3.0184 1.6 0.0001>, rotation=<rot 90 0 0>}
  {shapeType=Capsule, length=4.9607, radius=0.7611, center=<3.0287 -1.2889 0>, rotation=<rot 90 0 0>}
 ]
 dyn_chest_cloth_r_02 [{shapeType=Box, extentX=5.2781, extentY=3.5227, extentZ=1.2052, center=<2.9078 0 -0.0002>}]
 dyn_chest_cloth_r_03 [{shapeType=Box, extentX=6.67, extentY=1.9739, extentZ=1.2052, center=<3.6069 0 -0.0002>}]
 dyn_necklace_01 [{shapeType=Capsule, length=2.5, radius=2.2, center=<2 0 0>, rotation=<rot 90 0 0>}]
 dyn_necklace_02 [{shapeType=Capsule, length=2.5, radius=2.2, center=<3.7334 0 -0.0001>, rotation=<rot 90 0 0>}]
 dyn_necklace_03 [{shapeType=Capsule, length=2.5, radius=2.2, center=<2 0 0>, rotation=<rot 90 0 0>}]
 dyn_necklace_bk_01 [{shapeType=Capsule, length=2.5, radius=2.2, center=<2 0 0>, rotation=<rot 90 0 0>}]
 dyn_necklace_bk_02 [{shapeType=Capsule, length=3.75, radius=2.3, center=<2 0 0>, rotation=<rot 90 0 0>}]
 dyn_necklace_bk_03 [{shapeType=Capsule, length=3.25, radius=2.3, center=<2 0 0>, rotation=<rot 90 0 0>}]
 dyn_necklace_bk_04 [{shapeType=Capsule, length=4, radius=2.3, center=<2 0 0>, rotation=<rot 90 0 0>}]
 dyn_necklace_bk_05 [{shapeType=Capsule, length=3.5, radius=2.3, center=<2 0 0>, rotation=<rot 90 0 0>}]
 dyn_necklace_bk_06 [{shapeType=Capsule, length=2.5, radius=2.3, center=<2 0 0>, rotation=<rot 90 0 0>}]
 dyn_necklace_l_01 [{shapeType=Capsule, length=2.5, radius=2.2, center=<2 0 0>, rotation=<rot 90 0 0>}]
 dyn_necklace_l_02 [{shapeType=Capsule, length=2.5, radius=2.2, center=<2 0 0>, rotation=<rot 90 0 0>}]
 dyn_necklace_l_03 [{shapeType=Capsule, length=2.5, radius=2.2, center=<2 0 0>, rotation=<rot 90 0 0>}]
 dyn_necklace_l_04 [{shapeType=Capsule, length=2.5, radius=2.2, center=<2 0 0>, rotation=<rot 90 0 0>}]
 dyn_necklace_l_05 [{shapeType=Capsule, length=2.5, radius=2.2, center=<2 0 0>, rotation=<rot 90 0 0>}]
 dyn_necklace_l_06 [{shapeType=Capsule, length=2.5, radius=2.2, center=<2 0 0>, rotation=<rot 90 0 0>}]
 dyn_necklace_pin_l [{shapeType=Capsule, length=2.2048, radius=2.4952, rotation=<rot 90 0 0>}]
 dyn_necklace_pin_l_MSTR [{shapeType=Capsule, length=0.5, radius=0.5}]
 dyn_necklace_pin_r [{shapeType=Capsule, length=3.8621, radius=2.5879, rotation=<rot 74.5521 -0.0001 -0.0001>}]
 dyn_necklace_pin_r_MSTR [{shapeType=Capsule, length=0.5, radius=0.5}]
 dyn_necklace_r_01 [{shapeType=Capsule, length=2.5, radius=2.2, center=<2 0 0>, rotation=<rot 90 0 0>}]
 dyn_necklace_r_02 [{shapeType=Capsule, length=2.5, radius=2.2, center=<2 0 0>, rotation=<rot 90 0 0>}]
 dyn_necklace_r_03 [{shapeType=Capsule, length=2.5, radius=2.2, center=<2 0 0>, rotation=<rot 90 0 0>}]
 dyn_necklace_r_04 [{shapeType=Capsule, length=2.5, radius=2.2, center=<2 0 0>, rotation=<rot 90 0 0>}]
 dyn_necklace_r_05 [{shapeType=Capsule, length=2.5, radius=2.2, center=<2 0 0>, rotation=<rot 90 0 0>}]
 dyn_necklace_r_06 [{shapeType=Capsule, length=2.5, radius=2.2, center=<2 0 0>, rotation=<rot 90 0 0>}]
 neck_01 [{shapeType=Capsule, length=13.1149, radius=4.7564, center=<9.0898 -0.9485 0.0001>, rotation=<rot 90 180 -175.6909>}]
 spine_02 [{shapeType=Capsule, length=0.1727, radius=11.2146, center=<2.6698 1.7819 0>}]
 spine_03 [{shapeType=Capsule, length=3.1252, radius=11.3124, center=<2.2877 1.8564 0>}]
 spine_04 [{shapeType=Capsule, length=7.777, radius=11.6919, center=<3.0336 0.6597 0>}]
 spine_05 [
  {shapeType=Capsule, length=16.2708, radius=9.2657, center=<3.8549 -1.2455 0>}
  {shapeType=Capsule, length=16.4265, radius=5.7175, center=<7.018 -3.7593 0.0001>}
  {shapeType=Capsule, length=15.654, radius=9.2657, center=<-1.1877 1.9023 0>}
  {shapeType=Box, extentX=7.5, extentY=7.5, extentZ=19.2099, center=<10.7005 -7.7051 -0.0001>, rotation=<rot 0 12.9067 0>}
 ]
}
constraints [
 {bone1Name=dyn_chest_cloth_l_02, bone2Name=dyn_chest_cloth_l_01, swing1LimitDegrees=45, swing2LimitDegrees=45, twistLimitDegrees=15}
 {bone1Name=dyn_chest_cloth_l_03, bone2Name=dyn_chest_cloth_l_02, swing1LimitDegrees=45, swing2LimitDegrees=45, twistLimitDegrees=15}
 {bone1Name=dyn_chest_cloth_r_02, bone2Name=dyn_chest_cloth_r_01, swing1LimitDegrees=45, swing2LimitDegrees=45, twistLimitDegrees=5}
 {bone1Name=dyn_chest_cloth_r_03, bone2Name=dyn_chest_cloth_r_02, swing1LimitDegrees=45, swing2LimitDegrees=45, twistLimitDegrees=5}
 {bone1Name=dyn_necklace_02, bone2Name=dyn_necklace_01, swing1LimitDegrees=45, swing2LimitDegrees=45, twistLimitDegrees=5}
 {bone1Name=dyn_necklace_03, bone2Name=dyn_necklace_02, swing1LimitDegrees=45, swing2LimitDegrees=45, twistLimitDegrees=5}
 {bone1Name=dyn_necklace_l_02, bone2Name=dyn_necklace_l_01, swing1LimitDegrees=45, swing2LimitDegrees=45, twistLimitDegrees=5}
 {bone1Name=dyn_necklace_l_03, bone2Name=dyn_necklace_l_02, swing1LimitDegrees=45, swing2LimitDegrees=45, twistLimitDegrees=5}
 {bone1Name=dyn_necklace_l_04, bone2Name=dyn_necklace_l_03, swing1LimitDegrees=45, swing2LimitDegrees=45, twistLimitDegrees=5}
 {bone1Name=dyn_necklace_l_05, bone2Name=dyn_necklace_l_04, swing1LimitDegrees=45, swing2LimitDegrees=45, twistLimitDegrees=5}
 {bone1Name=dyn_necklace_l_06, bone2Name=dyn_necklace_l_05, swing1LimitDegrees=45, swing2LimitDegrees=45, twistLimitDegrees=5}
 {
  bone1Name=dyn_necklace_bk_02
  bone2Name=dyn_necklace_bk_01
  swing1LimitDegrees=60
  swing1Motion=Free
  swing2LimitDegrees=60
  swing2Motion=Free
  twistLimitDegrees=30
  twistMotion=Free
 }
 {
  bone1Name=dyn_necklace_bk_03
  bone2Name=dyn_necklace_bk_02
  swing1LimitDegrees=60
  swing1Motion=Free
  swing2LimitDegrees=60
  swing2Motion=Free
  twistLimitDegrees=30
  twistMotion=Free
 }
 {
  bone1Name=dyn_necklace_bk_04
  bone2Name=dyn_necklace_bk_03
  swing1LimitDegrees=60
  swing1Motion=Free
  swing2LimitDegrees=60
  swing2Motion=Free
  twistLimitDegrees=30
  twistMotion=Free
 }
 {
  bone1Name=dyn_necklace_bk_05
  bone2Name=dyn_necklace_bk_04
  swing1LimitDegrees=60
  swing1Motion=Free
  swing2LimitDegrees=60
  swing2Motion=Free
  twistLimitDegrees=30
  twistMotion=Free
 }
 {
  bone1Name=dyn_necklace_bk_06
  bone2Name=dyn_necklace_bk_05
  swing1LimitDegrees=60
  swing1Motion=Free
  swing2LimitDegrees=60
  swing2Motion=Free
  twistLimitDegrees=30
  twistMotion=Free
 }
 {bone1Name=dyn_necklace_pin_l, bone2Name=dyn_necklace_pin_l_MSTR, swing1LimitDegrees=45, swing2LimitDegrees=45, twistLimitDegrees=5}
 {bone1Name=dyn_necklace_r_02, bone2Name=dyn_necklace_r_01, swing1LimitDegrees=45, swing2LimitDegrees=45, twistLimitDegrees=5}
 {bone1Name=dyn_necklace_r_03, bone2Name=dyn_necklace_r_02, swing1LimitDegrees=45, swing2LimitDegrees=45, twistLimitDegrees=5}
 {bone1Name=dyn_necklace_r_04, bone2Name=dyn_necklace_r_03, swing1LimitDegrees=45, swing2LimitDegrees=45, twistLimitDegrees=5}
 {bone1Name=dyn_necklace_r_05, bone2Name=dyn_necklace_r_04, swing1LimitDegrees=45, swing2LimitDegrees=45, twistLimitDegrees=5}
 {bone1Name=dyn_necklace_r_06, bone2Name=dyn_necklace_r_05, swing1LimitDegrees=45, swing2LimitDegrees=45, twistLimitDegrees=5}
 {
  bone1Name=dyn_necklace_pin_r
  bone2Name=dyn_necklace_pin_r_MSTR
  swing1LimitDegrees=45
  swing1Motion=Free
  swing2LimitDegrees=45
  swing2Motion=Free
  twistLimitDegrees=5
  twistMotion=Free
 }
 {
  bone1Name=dyn_belt_high_01
  bone2Name=dyn_belt_high_MSTR
  swing1LimitDegrees=45
  swing1Motion=Free
  swing2LimitDegrees=45
  swing2Motion=Free
  twistLimitDegrees=45
  twistMotion=Free
 }
 {bone1Name=dyn_belt_high_02, bone2Name=dyn_belt_high_01, swing1LimitDegrees=5, swing2LimitDegrees=20, twistLimitDegrees=2}
 {bone1Name=dyn_belt_high_03, bone2Name=dyn_belt_high_02, swing1LimitDegrees=5, swing2LimitDegrees=20, twistLimitDegrees=2}
 {
  bone1Name=dyn_belt_low_01
  bone2Name=dyn_belt_low_MSTR
  swing1LimitDegrees=45
  swing1Motion=Free
  swing2LimitDegrees=45
  swing2Motion=Free
  twistLimitDegrees=45
  twistMotion=Free
 }
 {bone1Name=dyn_belt_low_02, bone2Name=dyn_belt_low_01, swing1LimitDegrees=5, swing2LimitDegrees=20, twistLimitDegrees=2}
 {bone1Name=dyn_belt_low_03, bone2Name=dyn_belt_low_02, swing1LimitDegrees=5, swing2LimitDegrees=20, twistLimitDegrees=2}
 {bone1Name=dyn_chest_cloth_l_01, bone2Name=spine_05, swing1LimitDegrees=45, swing2LimitDegrees=45, twistLimitDegrees=15}
 {bone1Name=dyn_chest_cloth_r_01, bone2Name=spine_05, swing1LimitDegrees=45, swing2LimitDegrees=45, twistLimitDegrees=5}
 {bone1Name=dyn_necklace_01, bone2Name=spine_05, swing1LimitDegrees=45, swing2LimitDegrees=45, twistLimitDegrees=5}
 {bone1Name=dyn_necklace_l_01, bone2Name=dyn_necklace_pin_l, swing1LimitDegrees=45, swing2LimitDegrees=45, twistLimitDegrees=5}
 {bone1Name=dyn_necklace_bk_01, bone2Name=dyn_necklace_pin_l, swing1LimitDegrees=10, swing2LimitDegrees=10, twistLimitDegrees=30}
 {
  bone1Name=dyn_necklace_pin_l_MSTR
  bone2Name=spine_05
  swing1LimitDegrees=45
  swing1Motion=Free
  swing2LimitDegrees=45
  swing2Motion=Free
  twistLimitDegrees=45
  twistMotion=Free
 }
 {bone1Name=dyn_necklace_r_01, bone2Name=dyn_necklace_pin_r, swing1LimitDegrees=45, swing2LimitDegrees=45, twistLimitDegrees=5}
 {
  bone1Name=dyn_necklace_pin_r_MSTR
  bone2Name=spine_05
  swing1LimitDegrees=45
  swing1Motion=Free
  swing2LimitDegrees=45
  swing2Motion=Free
  twistLimitDegrees=45
  twistMotion=Free
 }
 {
  bone1Name=spine_05
  bone2Name=spine_04
  swing1LimitDegrees=45
  swing1Motion=Free
  swing2LimitDegrees=45
  swing2Motion=Free
  twistLimitDegrees=45
  twistMotion=Free
 }
 {
  bone1Name=spine_04
  bone2Name=spine_03
  swing1LimitDegrees=45
  swing1Motion=Free
  swing2LimitDegrees=45
  swing2Motion=Free
  twistLimitDegrees=45
  twistMotion=Free
 }
 {
  bone1Name=spine_03
  bone2Name=spine_02
  swing1LimitDegrees=45
  swing1Motion=Free
  swing2LimitDegrees=45
  swing2Motion=Free
  twistLimitDegrees=45
  twistMotion=Free
 }
 {
  bone1Name=clavicle_l
  bone2Name=spine_05
  swing1LimitDegrees=45
  swing1Motion=Free
  swing2LimitDegrees=45
  swing2Motion=Free
  twistLimitDegrees=45
  twistMotion=Free
 }
 {
  bone1Name=clavicle_r
  bone2Name=spine_05
  swing1LimitDegrees=45
  swing1Motion=Free
  swing2LimitDegrees=45
  swing2Motion=Free
  twistLimitDegrees=45
  twistMotion=Free
 }
 {
  bone1Name=dyn_necklace_pin_r
  bone2Name=dyn_necklace_bk_06
  swing1LimitDegrees=45
  swing1Motion=Free
  swing2LimitDegrees=45
  swing2Motion=Free
  twistLimitDegrees=45
  twistMotion=Free
 }
 {
  bone1Name=dyn_necklace_03
  bone2Name=dyn_necklace_r_06
  swing1LimitDegrees=45
  swing1Motion=Free
  swing2LimitDegrees=45
  swing2Motion=Free
  twistLimitDegrees=45
  twistMotion=Free
 }
 {
  bone1Name=dyn_necklace_03
  bone2Name=dyn_necklace_l_06
  swing1LimitDegrees=45
  swing1Motion=Free
  swing2LimitDegrees=45
  swing2Motion=Free
  twistLimitDegrees=45
  twistMotion=Free
 }
 {
  bone1Name=neck_01
  bone2Name=spine_05
  swing1LimitDegrees=45
  swing1Motion=Free
  swing2LimitDegrees=45
  swing2Motion=Free
  twistLimitDegrees=45
  twistMotion=Free
 }
]
