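# Anim Blueprint for the Golden Skeleton player body cosmetic, adding secondary motion to skirt, necklace and belts.
# It copies the attached parent mesh's pose, then adjusts wrists, crouch collision and shoulder/necklace anchors.
# From a cached base pose, skydiving and normal skirt branches mix RigidBody with bone constraints (plus trails when normal), then add torso physics, necklace constraints, AnimDynamics belts and acceleration tilts.
# Emotes bypass that chain for a dedicated skirt simulation.
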AnimBlueprint /BRCosmetics/Characters/Player/Male/Medium/Bodies/M_MED_The_Golden_Skeleton/Meshes/M_MED_The_Golden_Skeleton_AnimBP
target_skeleton=&/M_MED_The_Golden_Skeleton_Skeleton
graphs [
 AnimGraph {nodes [
   n1 = RigidBody!(ComponentPose, Alpha=0, OverrideWorldGravity:Vector="0,0,0", ExternalForce) -> (Pose) {
    bClampLinearTranslationLimitToRefPose=true
    bOverrideWorldGravity=true
    overridePhysicsAsset=/Game/Characters/Player/Male/Medium/Bodies/M_MED_The_Golden_Skeleton/Meshes/M_MED_The_Golden_Skeleton_Dyn_skirt_skydiving
    simulationSpace=BaseBoneSpace
    alphaBoolBlend {blendOption=Linear}
    baseBoneRef {boneName=pelvis}
    componentAppliedLinearAccClamp=<3000 3000 3000>
    componentLinearAccScale=<0.4 0.3 0.3>
    componentLinearVelScale=<0.4 0.3 0.3>
    overrideWorldGravity {z=-980}
   }
   n2 = RigidBody(ComponentPose=@n1, Alpha=0, OverrideWorldGravity="0,0,-980") {
    bClampLinearTranslationLimitToRefPose=true
    bOverrideWorldGravity=true
    overridePhysicsAsset=/Game/Characters/Player/Male/Medium/Bodies/M_MED_The_Golden_Skeleton/Meshes/M_MED_The_Golden_Skeleton_Dyn_torso_skydiving
    simulationSpace=BaseBoneSpace
    alphaBoolBlend {blendOption=Linear}
    baseBoneRef {boneName=spine_05}
    componentAppliedLinearAccClamp=<3000 3000 3000>
    componentLinearAccScale=<0.4 0.3 0.3>
    componentLinearVelScale=<0.4 0.3 0.3>
    overrideWorldGravity {z=-980}
   }
   Component To Local(ComponentPose=@n2)
   Component To Local()
   "Transform (Modify) Bone - Bone: dyn_cloth_bk_01"(Alpha=0, Translation="0,0,0", Rotation="60,0,0", Scale="1,1,1") {
    rotationMode=BMM_Additive
    rotationSpace=BCS_BoneSpace
    scaleMode=BMM_Additive
    scaleSpace=BCS_BoneSpace
    translationMode=BMM_Additive
    translationSpace=BCS_BoneSpace
    alphaBoolBlend {blendOption=Linear}
    boneToModify {boneName=dyn_cloth_bk_01}
    rotation {pitch=60}
   }
   n3 = RigidBody!(ComponentPose, Alpha, ExternalForce) -> (Pose) {
    bClampLinearTranslationLimitToRefPose=true
    bOverrideWorldGravity=true
    overridePhysicsAsset=/Game/Characters/Player/Male/Medium/Bodies/M_MED_The_Golden_Skeleton/Meshes/M_MED_The_Golden_Skeleton_Dyn_arm_l
    simulationSpace=BaseBoneSpace
    alphaBoolBlend {blendOption=Linear}
    baseBoneRef {boneName=clavicle_l}
    componentAppliedLinearAccClamp=<3000 3000 3000>
    componentLinearAccScale=<0.4 0.3 0.3>
    componentLinearVelScale=<0.4 0.3 0.3>
    overrideWorldGravity {z=-980}
   }
   n4 = RigidBody(ComponentPose=@n3) {
    bClampLinearTranslationLimitToRefPose=true
    bOverrideWorldGravity=true
    overridePhysicsAsset=/Game/Characters/Player/Male/Medium/Bodies/M_MED_The_Golden_Skeleton/Meshes/M_MED_The_Golden_Skeleton_Dyn_arm_r
    simulationSpace=BaseBoneSpace
    alphaBoolBlend {blendOption=Linear}
    baseBoneRef {boneName=upperarm_r}
    componentAppliedLinearAccClamp=<3000 3000 3000>
    componentLinearAccScale=<0.4 0.3 0.3>
    componentLinearVelScale=<0.4 0.3 0.3>
    overrideWorldGravity {z=-980}
   }
   RigidBody(ComponentPose=@n4) {
    bClampLinearTranslationLimitToRefPose=true
    bOverrideWorldGravity=true
    overridePhysicsAsset=/Game/Characters/Player/Male/Medium/Bodies/M_MED_The_Golden_Skeleton/Meshes/M_MED_The_Golden_Skeleton_Dyn_leg_r
    simulationSpace=BaseBoneSpace
    alphaBoolBlend {blendOption=Linear}
    baseBoneRef {boneName=thigh_r}
    componentAppliedLinearAccClamp=<3000 3000 3000>
    componentLinearAccScale=<0.4 0.3 0.3>
    componentLinearVelScale=<0.4 0.3 0.3>
    overrideWorldGravity {z=-980}
   }
   n5 = RigidBody!(ComponentPose=Local To Component(LocalPose=Use cached pose 'base Pose'()), bAlphaBoolEnabled:bool, ExternalForce) -> (Pose) {
    alphaInputType=Bool
    bClampLinearTranslationLimitToRefPose=true
    bOverrideWorldGravity=true
    overridePhysicsAsset=&/M_MED_The_Golden_Skeleton_Dyn_skirt_EMOTE
    simulationSpace=BaseBoneSpace
    alphaBoolBlend {blendOption=Linear}
    baseBoneRef {boneName=pelvis}
    componentAppliedLinearAccClamp=<3000 3000 3000>
    componentLinearAccScale=<0.7 0.5 0.5>
    componentLinearVelScale=<0.7 0.5 0.5>
    overrideWorldGravity {z=-980}
   }
   n6 = "Constraint - dyn_cloth_r_01 (P)"(ComponentPose=Local To Component(LocalPose=Use cached pose 'base Pose'()), Alpha=1, ConstraintWeights_0=1) {
    alphaBoolBlend {blendOption=Linear}
    boneToModify {boneName=dyn_cloth_r_01}
    constraintSetup [{targetBone {boneName=thigh_twist_01_r}}]
    constraintWeights [1]
   }
   n7 = Get thigh_shape_fn_l() -> (thigh_shape_fn_l:double)
   n8 = "Constraint - dyn_cloth_fr_l_MSTR (RR)"!(ComponentPose=@n6, Alpha=@n7, ConstraintWeights_0=0.9, ConstraintWeights_1:float=0.1) -> (Pose) {
    alphaBoolBlend {blendOption=Linear}
    boneToModify {boneName=dyn_cloth_fr_l_MSTR}
    constraintSetup [{transformType=Rotation, targetBone {boneName=thigh_twist_01_l}}, {transformType=Rotation, targetBone {boneName=thigh_twist_01_r}}]
    constraintWeights [1,1]
   }
   n9 = Get thigh_shape_fn_r() -> (thigh_shape_fn_r:double)
   n10 = "Constraint - dyn_cloth_fr_r_MSTR (RR)"!(ComponentPose=@n8, Alpha=@n9, ConstraintWeights_0=0.1, ConstraintWeights_1:float=0.9) -> (Pose) {
    alphaBoolBlend {blendOption=Linear}
    boneToModify {boneName=dyn_cloth_fr_r_MSTR}
    constraintSetup [{transformType=Rotation, targetBone {boneName=thigh_twist_01_l}}, {transformType=Rotation, targetBone {boneName=thigh_twist_01_r}}]
    constraintWeights [1,1]
   }
   n11 = "Transform (Modify) Bone - Bone: dyn_cloth_fr_l_MSTR"(ComponentPose=@n10, Alpha=@n7, Translation="0,-10,15", Rotation="-20,10,30", Scale="1,1,1") {
    rotationMode=BMM_Additive
    rotationSpace=BCS_BoneSpace
    scaleMode=BMM_Additive
    scaleSpace=BCS_BoneSpace
    translationMode=BMM_Additive
    translationSpace=BCS_BoneSpace
    alphaBoolBlend {blendOption=Linear}
    boneToModify {boneName=dyn_cloth_fr_l_MSTR}
    rotation=<rot -20 10 30>
    translation {y=-10, z=15}
   }
   n12 = "Transform (Modify) Bone - Bone: dyn_belt_thigh_r"(ComponentPose=@n11, Alpha=@n9, Translation="-15,0,0", Rotation="0,0,0", Scale="1,1,1") {
    rotationMode=BMM_Additive
    rotationSpace=BCS_BoneSpace
    scaleMode=BMM_Additive
    scaleSpace=BCS_BoneSpace
    translationMode=BMM_Additive
    translationSpace=BCS_BoneSpace
    alphaBoolBlend {blendOption=Linear}
    boneToModify {boneName=dyn_belt_thigh_r}
    translation {x=-15}
   }
   n13 = "Transform (Modify) Bone - Bone: dyn_cloth_fr_r_MSTR"(ComponentPose=@n12, Alpha=@n9, Translation="0,10,15", Rotation="-20,-10,30", Scale="1,1,1") {
    rotationMode=BMM_Additive
    rotationSpace=BCS_BoneSpace
    scaleMode=BMM_Additive
    scaleSpace=BCS_BoneSpace
    translationMode=BMM_Additive
    translationSpace=BCS_BoneSpace
    alphaBoolBlend {blendOption=Linear}
    boneToModify {boneName=dyn_cloth_fr_r_MSTR}
    rotation=<rot -20 -10 30>
    translation {y=10, z=15}
   }
   n14 = Trail controller - Bone: dyn_cloth_l_06(ComponentPose=@n13, Alpha=0.7) {
    bLimitRotation=true
    bUsePlanarLimit=true
    chainLength=6
    debugLifeTime=0.1
    maxDeltaTime=0.333
    alphaScaleBiasClamp {bInterpResult=true, scale=0.65}
    baseJoint {boneName=root}
    rotationLimits [
     {limitMax=<180 180 180>}
     {limitMax=<180 180 180>}
     {limitMax=<180 180 180>}
     {limitMax=<180 180 180>}
     {limitMax=<180 180 180>}
     {limitMax=<180 180 180>}
    ]
    rotationOffsets [<0 0 0>, <0 0 0>, <0 0 0>, <0 0 0>, <0 0 0>, <0 0 0>]
    trailBone {boneName=dyn_cloth_l_06}
    trailRelaxationSpeed {editorCurveData {keys [<at -0.0134 19.3841 RCIM_Linear>,<at 0.4187 12.9226 RCIM_Linear>,<at 0.9955 9.6044 RCIM_Linear>]}}
   }
   n15 = Trail controller - Bone: dyn_cloth_bk_05(ComponentPose=@n14, Alpha=0.5) {
    bLimitRotation=true
    bUsePlanarLimit=true
    chainLength=4
    debugLifeTime=0.1
    maxDeltaTime=0.333
    alphaScaleBiasClamp {bInterpResult=true, scale=0.65}
    baseJoint {boneName=root}
    rotationLimits [{limitMax=<180 180 180>}, {limitMax=<180 180 180>}, {limitMax=<180 180 180>}, {limitMax=<180 180 180>}]
    rotationOffsets [<0 0 0>, <0 0 0>, <0 0 0>, <0 0 0>]
    trailBone {boneName=dyn_cloth_bk_05}
    trailRelaxationSpeed {editorCurveData {keys [<at -0.0134 19.3841 RCIM_Linear>,<at 0.4187 12.9226 RCIM_Linear>,<at 0.9955 9.6044 RCIM_Linear>]}}
   }
   n16 = Trail controller - Bone: dyn_cloth_r_03(ComponentPose=@n15, Alpha=0.5) {
    bLimitRotation=true
    bUsePlanarLimit=true
    debugLifeTime=0.1
    maxDeltaTime=0.333
    alphaScaleBiasClamp {bInterpResult=true, scale=0.65}
    baseJoint {boneName=root}
    rotationLimits [{limitMax=<180 180 180>}, {limitMax=<180 180 180>}]
    rotationOffsets [<0 0 0>, <0 0 0>]
    trailBone {boneName=dyn_cloth_r_03}
    trailRelaxationSpeed {editorCurveData {keys [<at -0.0134 19.3841 RCIM_Linear>,<at 0.4187 12.9226 RCIM_Linear>,<at 0.9955 9.6044 RCIM_Linear>]}}
   }
   n17 = RigidBody!(ComponentPose=@n16, bAlphaBoolEnabled, OverrideWorldGravity:Vector=Get gravity_override() -> (gravity_override:Vector), ExternalForce) -> (Pose) {
    alphaInputType=Bool
    bClampLinearTranslationLimitToRefPose=true
    bOverrideWorldGravity=true
    overridePhysicsAsset=&/M_MED_The_Golden_Skeleton_Dyn_skirt
    simulationSpace=BaseBoneSpace
    alphaBoolBlend {blendOption=Linear}
    baseBoneRef {boneName=pelvis}
    componentAppliedLinearAccClamp=<1200 1200 1200>
    componentLinearAccScale=<0.28 0.28 0.3>
    componentLinearVelScale=<0.35 0.35 0.35>
    overrideWorldGravity {z=-980}
   }
   n18 = "Constraint - dyn_cloth_bk_05 (PP)"!(ComponentPose=@n17, Alpha=0.7, ConstraintWeights_0=0.5, ConstraintWeights_1:float=0.5) -> (Pose) {
    alphaBoolBlend {blendOption=Linear}
    boneToModify {boneName=dyn_cloth_bk_05}
    constraintSetup [{targetBone {boneName=dyn_cloth_l_05}}, {targetBone {boneName=dyn_cloth_bk_05}}]
    constraintWeights [1,1]
   }
   n19 = "Constraint - dyn_cloth_bk_r_02 (PP)"!(ComponentPose=@n18, Alpha=0.7, ConstraintWeights_0=0.5, ConstraintWeights_1:float=0.5) -> (Pose) {
    alphaBoolBlend {blendOption=Linear}
    boneToModify {boneName=dyn_cloth_bk_r_02}
    constraintSetup [{targetBone {boneName=dyn_cloth_r_02}}, {targetBone {boneName=dyn_cloth_bk_02}}]
    constraintWeights [1,1]
   }
   n20 = "Constraint - dyn_cloth_bk_r_03 (PP)"!(ComponentPose=@n19, Alpha=0.7, ConstraintWeights_0=0.5, ConstraintWeights_1:float=0.5) -> (Pose) {
    alphaBoolBlend {blendOption=Linear}
    boneToModify {boneName=dyn_cloth_bk_r_03}
    constraintSetup [{targetBone {boneName=dyn_cloth_r_03}}, {targetBone {boneName=dyn_cloth_bk_03}}]
    constraintWeights [1,1]
   }
   n21 = "Constraint - dyn_cloth_bk_l_01 (PP)"!(ComponentPose=@n20, Alpha=0.7, ConstraintWeights_0=0.5, ConstraintWeights_1:float=0.5) -> (Pose) {
    alphaBoolBlend {blendOption=Linear}
    boneToModify {boneName=dyn_cloth_bk_l_01}
    constraintSetup [{targetBone {boneName=dyn_cloth_bk_01}}, {targetBone {boneName=dyn_cloth_l_01}}]
    constraintWeights [1,1]
   }
   n22 = "Constraint - dyn_cloth_bk_l_02 (PP)"!(ComponentPose=@n21, Alpha=0.7, ConstraintWeights_0=0.5, ConstraintWeights_1:float=0.5) -> (Pose) {
    alphaBoolBlend {blendOption=Linear}
    boneToModify {boneName=dyn_cloth_bk_l_02}
    constraintSetup [{targetBone {boneName=dyn_cloth_bk_02}}, {targetBone {boneName=dyn_cloth_l_02}}]
    constraintWeights [1,1]
   }
   n23 = "Constraint - dyn_cloth_bk_l_03 (PP)"!(ComponentPose=@n22, Alpha=0.7, ConstraintWeights_0=0.5, ConstraintWeights_1:float=0.5) -> (Pose) {
    alphaBoolBlend {blendOption=Linear}
    boneToModify {boneName=dyn_cloth_bk_l_03}
    constraintSetup [{targetBone {boneName=dyn_cloth_bk_03}}, {targetBone {boneName=dyn_cloth_l_03}}]
    constraintWeights [1,1]
   }
   n24 = "Constraint - dyn_cloth_bk_l_04 (PP)"!(ComponentPose=@n23, Alpha=0.7, ConstraintWeights_0=0.5, ConstraintWeights_1:float=0.5) -> (Pose) {
    alphaBoolBlend {blendOption=Linear}
    boneToModify {boneName=dyn_cloth_bk_l_04}
    constraintSetup [{targetBone {boneName=dyn_cloth_bk_04}}, {targetBone {boneName=dyn_cloth_l_04}}]
    constraintWeights [1,1]
   }
   n25 = "Constraint - dyn_cloth_bk_l_05 (PP)"!(ComponentPose=@n24, Alpha=0.7, ConstraintWeights_0=0.5, ConstraintWeights_1:float=0.5) -> (Pose) {
    alphaBoolBlend {blendOption=Linear}
    boneToModify {boneName=dyn_cloth_bk_l_05}
    constraintSetup [{targetBone {boneName=dyn_cloth_bk_05}}, {targetBone {boneName=dyn_cloth_l_05}}]
    constraintWeights [1,1]
   }
   n26 = "Constraint - dyn_cloth_fr_r_01 (PP)"!(ComponentPose=@n25, Alpha=0.7, ConstraintWeights_0=0.9, ConstraintWeights_1:float=0.1) -> (Pose) {
    alphaBoolBlend {blendOption=Linear}
    boneToModify {boneName=dyn_cloth_fr_r_01}
    constraintSetup [{targetBone {boneName=dyn_cloth_fr_r_01}}, {targetBone {boneName=dyn_cloth_fr_l_01}}]
    constraintWeights [1,1]
   }
   n27 = "Constraint - dyn_cloth_fr_r_02 (PP)"!(ComponentPose=@n26, Alpha=0.7, ConstraintWeights_0=0.9, ConstraintWeights_1:float=0.1) -> (Pose) {
    alphaBoolBlend {blendOption=Linear}
    boneToModify {boneName=dyn_cloth_fr_r_02}
    constraintSetup [{targetBone {boneName=dyn_cloth_fr_r_02}}, {targetBone {boneName=dyn_cloth_fr_l_02}}]
    constraintWeights [1,1]
   }
   n28 = "Constraint - dyn_cloth_fr_r_03 (PP)"!(ComponentPose=@n27, Alpha=0.7, ConstraintWeights_0=0.9, ConstraintWeights_1:float=0.1) -> (Pose) {
    alphaBoolBlend {blendOption=Linear}
    boneToModify {boneName=dyn_cloth_fr_r_03}
    constraintSetup [{targetBone {boneName=dyn_cloth_fr_r_03}}, {targetBone {boneName=dyn_cloth_fr_l_03}}]
    constraintWeights [1,1]
   }
   n29 = "Constraint - dyn_cloth_fr_l_01 (PP)"!(ComponentPose=@n28, Alpha=0.7, ConstraintWeights_0=0.9, ConstraintWeights_1:float=0.1) -> (Pose) {
    alphaBoolBlend {blendOption=Linear}
    boneToModify {boneName=dyn_cloth_fr_l_01}
    constraintSetup [{targetBone {boneName=dyn_cloth_fr_l_01}}, {targetBone {boneName=dyn_cloth_fr_l_MSTR}}]
    constraintWeights [1,1]
   }
   n30 = "Constraint - dyn_cloth_fr_l_02 (PP)"!(ComponentPose=@n29, Alpha=0.7, ConstraintWeights_0=0.9, ConstraintWeights_1:float=0.1) -> (Pose) {
    alphaBoolBlend {blendOption=Linear}
    boneToModify {boneName=dyn_cloth_fr_l_02}
    constraintSetup [{targetBone {boneName=dyn_cloth_fr_l_02}}, {targetBone {boneName=dyn_cloth_fr_l_MSTR}}]
    constraintWeights [1,1]
   }
   n31 = "Constraint - dyn_cloth_fr_l_03 (PP)"!(ComponentPose=@n30, Alpha=0.7, ConstraintWeights_0=0.9, ConstraintWeights_1:float=0.1) -> (Pose) {
    alphaBoolBlend {blendOption=Linear}
    boneToModify {boneName=dyn_cloth_fr_l_03}
    constraintSetup [{targetBone {boneName=dyn_cloth_fr_l_03}}, {targetBone {boneName=dyn_cloth_fr_l_MSTR}}]
    constraintWeights [1,1]
   }
   n32 = Component To Local(ComponentPose=@n31)
   n33 = RigidBody(ComponentPose=Local To Component(LocalPose=Use cached pose 'base Pose'()), OverrideWorldGravity=Get gravity_override()) {
    alphaInputType=Bool
    bClampLinearTranslationLimitToRefPose=true
    bOverrideWorldGravity=true
    overridePhysicsAsset=&M_MED_The_Golden_Skeleton_Dyn_skirt
    simulationSpace=BaseBoneSpace
    alphaBoolBlend {blendOption=Linear}
    baseBoneRef {boneName=pelvis}
    componentAppliedLinearAccClamp=<1200 1200 1200>
    componentLinearAccScale=<0.28 0.28 0.3>
    componentLinearVelScale=<0.35 0.35 0.35>
    overrideWorldGravity {z=-980}
   }
   n34 = "Constraint - dyn_cloth_r_02 (P)"(ComponentPose=@n33, Alpha=0.7, ConstraintWeights_0=1) {
    alphaBoolBlend {blendOption=Linear}
    boneToModify {boneName=dyn_cloth_r_02}
    constraintSetup [{targetBone {boneName=thigh_r}}]
    constraintWeights [1]
   }
   n35 = "Constraint - dyn_cloth_r_03 (P)"(ComponentPose=@n34, Alpha=0.4, ConstraintWeights_0=1) {
    alphaBoolBlend {blendOption=Linear}
    boneToModify {boneName=dyn_cloth_r_03}
    constraintSetup [{targetBone {boneName=thigh_r}}]
    constraintWeights [1]
   }
   n36 = "Constraint - dyn_cloth_bk_r_02 (P)"(ComponentPose=@n35, Alpha=0.7, ConstraintWeights_0=1) {
    alphaBoolBlend {blendOption=Linear}
    boneToModify {boneName=dyn_cloth_bk_r_02}
    constraintSetup [{targetBone {boneName=thigh_r}}]
    constraintWeights [1]
   }
   n37 = "Constraint - dyn_cloth_bk_r_03 (P)"(ComponentPose=@n36, Alpha=0.4, ConstraintWeights_0=1) {
    alphaBoolBlend {blendOption=Linear}
    boneToModify {boneName=dyn_cloth_bk_r_03}
    constraintSetup [{targetBone {boneName=thigh_r}}]
    constraintWeights [1]
   }
   n38 = "Constraint - dyn_cloth_bk_02 (PP)"!(ComponentPose=@n37, Alpha=1, ConstraintWeights_0=0.5, ConstraintWeights_1:float=0.5) -> (Pose) {
    alphaBoolBlend {blendOption=Linear}
    boneToModify {boneName=dyn_cloth_bk_02}
    constraintSetup [{targetBone {boneName=thigh_l}}, {targetBone {boneName=thigh_r}}]
    constraintWeights [1,1]
   }
   n39 = "Constraint - dyn_cloth_bk_03 (PP)"!(ComponentPose=@n38, Alpha=0.7, ConstraintWeights_0=0.5, ConstraintWeights_1:float=0.5) -> (Pose) {
    alphaBoolBlend {blendOption=Linear}
    boneToModify {boneName=dyn_cloth_bk_03}
    constraintSetup [{targetBone {boneName=thigh_l}}, {targetBone {boneName=thigh_r}}]
    constraintWeights [1,1]
   }
   n40 = "Constraint - dyn_cloth_bk_04 (PP)"!(ComponentPose=@n39, Alpha=0.4, ConstraintWeights_0=0.5, ConstraintWeights_1:float=0.5) -> (Pose) {
    alphaBoolBlend {blendOption=Linear}
    boneToModify {boneName=dyn_cloth_bk_04}
    constraintSetup [{targetBone {boneName=calf_l}}, {targetBone {boneName=calf_r}}]
    constraintWeights [1,1]
   }
   n41 = "Constraint - dyn_cloth_bk_05 (PP)"(ComponentPose=@n40, Alpha=0.2, ConstraintWeights_0=0.5, ConstraintWeights_1=0.5) {
    alphaBoolBlend {blendOption=Linear}
    boneToModify {boneName=dyn_cloth_bk_05}
    constraintSetup [{targetBone {boneName=calf_l}}, {targetBone {boneName=calf_r}}]
    constraintWeights [1,1]
   }
   n42 = "Constraint - dyn_cloth_l_02 (P)"(ComponentPose=@n41, Alpha=1, ConstraintWeights_0=1) {
    alphaBoolBlend {blendOption=Linear}
    boneToModify {boneName=dyn_cloth_l_02}
    constraintSetup [{targetBone {boneName=thigh_l}}]
    constraintWeights [1]
   }
   n43 = "Constraint - dyn_cloth_l_03 (P)"(ComponentPose=@n42, Alpha=0.7, ConstraintWeights_0=1) {
    alphaBoolBlend {blendOption=Linear}
    boneToModify {boneName=dyn_cloth_l_03}
    constraintSetup [{targetBone {boneName=thigh_l}}]
    constraintWeights [1]
   }
   n44 = "Constraint - dyn_cloth_l_04 (P)"(ComponentPose=@n43, Alpha=0.4, ConstraintWeights_0=1) {
    alphaBoolBlend {blendOption=Linear}
    boneToModify {boneName=dyn_cloth_l_04}
    constraintSetup [{targetBone {boneName=calf_l}}]
    constraintWeights [1]
   }
   n45 = "Constraint - dyn_cloth_l_05 (P)"(ComponentPose=@n44, Alpha=0.2, ConstraintWeights_0=1) {
    alphaBoolBlend {blendOption=Linear}
    boneToModify {boneName=dyn_cloth_l_05}
    constraintSetup [{targetBone {boneName=calf_l}}]
    constraintWeights [1]
   }
   n46 = "Constraint - dyn_cloth_l_06 (P)"(ComponentPose=@n45, Alpha=0.1, ConstraintWeights_0=1) {
    alphaBoolBlend {blendOption=Linear}
    boneToModify {boneName=dyn_cloth_l_06}
    constraintSetup [{targetBone {boneName=calf_l}}]
    constraintWeights [1]
   }
   n47 = "Constraint - dyn_cloth_bk_l_02 (P)"(ComponentPose=@n46, Alpha=1, ConstraintWeights_0=1) {
    alphaBoolBlend {blendOption=Linear}
    boneToModify {boneName=dyn_cloth_bk_l_02}
    constraintSetup [{targetBone {boneName=thigh_l}}]
    constraintWeights [1]
   }
   n48 = "Constraint - dyn_cloth_bk_l_03 (P)"(ComponentPose=@n47, Alpha=0.7, ConstraintWeights_0=1) {
    alphaBoolBlend {blendOption=Linear}
    boneToModify {boneName=dyn_cloth_bk_l_03}
    constraintSetup [{targetBone {boneName=thigh_l}}]
    constraintWeights [1]
   }
   n49 = "Constraint - dyn_cloth_bk_l_04 (P)"(ComponentPose=@n48, Alpha=0.4, ConstraintWeights_0=1) {
    alphaBoolBlend {blendOption=Linear}
    boneToModify {boneName=dyn_cloth_bk_l_04}
    constraintSetup [{targetBone {boneName=calf_l}}]
    constraintWeights [1]
   }
   n50 = "Constraint - dyn_cloth_bk_l_05 (P)"(ComponentPose=@n49, Alpha=0.2, ConstraintWeights_0=1) {
    alphaBoolBlend {blendOption=Linear}
    boneToModify {boneName=dyn_cloth_bk_l_05}
    constraintSetup [{targetBone {boneName=calf_l}}]
    constraintWeights [1]
   }
   n51 = "Constraint - dyn_cloth_fr_l_MSTR (P)"(ComponentPose=@n50, Alpha=0.5, ConstraintWeights_0=1) {
    alphaBoolBlend {blendOption=Linear}
    boneToModify {boneName=dyn_cloth_fr_l_MSTR}
    constraintSetup [{targetBone {boneName=thigh_l}}]
    constraintWeights [1]
   }
   n52 = "Constraint - dyn_cloth_fr_l_02 (P)"(ComponentPose=@n51, Alpha=0.5, ConstraintWeights_0=1) {
    alphaBoolBlend {blendOption=Linear}
    boneToModify {boneName=dyn_cloth_fr_l_02}
    constraintSetup [{targetBone {boneName=thigh_l}}]
    constraintWeights [1]
   }
   n53 = "Constraint - dyn_cloth_fr_l_03 (P)"(ComponentPose=@n52, Alpha=0.5, ConstraintWeights_0=1) {
    alphaBoolBlend {blendOption=Linear}
    boneToModify {boneName=dyn_cloth_fr_l_03}
    constraintSetup [{targetBone {boneName=thigh_l}}]
    constraintWeights [1]
   }
   n54 = "Constraint - dyn_cloth_fr_r_MSTR (P)"(ComponentPose=@n53, Alpha=0.5, ConstraintWeights_0=1) {
    alphaBoolBlend {blendOption=Linear}
    boneToModify {boneName=dyn_cloth_fr_r_MSTR}
    constraintSetup [{targetBone {boneName=thigh_r}}]
    constraintWeights [1]
   }
   n55 = "Constraint - dyn_cloth_fr_r_02 (P)"(ComponentPose=@n54, Alpha=0.5, ConstraintWeights_0=1) {
    alphaBoolBlend {blendOption=Linear}
    boneToModify {boneName=dyn_cloth_fr_r_02}
    constraintSetup [{targetBone {boneName=thigh_r}}]
    constraintWeights [1]
   }
   n56 = "Constraint - dyn_cloth_fr_r_03 (P)"(ComponentPose=@n55, Alpha=0.5, ConstraintWeights_0=1) {
    alphaBoolBlend {blendOption=Linear}
    boneToModify {boneName=dyn_cloth_fr_r_03}
    constraintSetup [{targetBone {boneName=thigh_r}}]
    constraintWeights [1]
   }
   n57 = "Transform (Modify) Bone - Bone: dyn_cloth_l_MSTR"(ComponentPose=@n56, Alpha=1, Translation="0,0,0", Rotation="0,5,0", Scale="1,1,1") {
    rotationMode=BMM_Additive
    rotationSpace=BCS_ParentBoneSpace
    translationMode=BMM_Additive
    translationSpace=BCS_ParentBoneSpace
    alphaBoolBlend {blendOption=Linear}
    boneToModify {boneName=dyn_cloth_l_MSTR}
    rotation {yaw=5}
   }
   n58 = "Transform (Modify) Bone - Bone: dyn_cloth_bk_l_01"(ComponentPose=@n57, Alpha=1, Translation="0,0,0", Rotation="0,5,0", Scale="1,1,1") {
    rotationMode=BMM_Additive
    rotationSpace=BCS_ParentBoneSpace
    translationMode=BMM_Additive
    translationSpace=BCS_ParentBoneSpace
    alphaBoolBlend {blendOption=Linear}
    boneToModify {boneName=dyn_cloth_bk_l_01}
    rotation {yaw=5}
   }
   n59 = "Transform (Modify) Bone - Bone: dyn_cloth_bk_01"(ComponentPose=@n58, Alpha=1, Translation="0,0,0", Rotation="0,5,0", Scale="1,1,1") {
    rotationMode=BMM_Additive
    rotationSpace=BCS_ParentBoneSpace
    translationMode=BMM_Additive
    translationSpace=BCS_ParentBoneSpace
    alphaBoolBlend {blendOption=Linear}
    boneToModify {boneName=dyn_cloth_bk_01}
    rotation {yaw=5}
   }
   n60 = "Transform (Modify) Bone - Bone: dyn_cloth_bk_r_01"(ComponentPose=@n59, Alpha=1, Translation="0,0,0", Rotation="0,5,0", Scale="1,1,1") {
    rotationMode=BMM_Additive
    rotationSpace=BCS_ParentBoneSpace
    translationMode=BMM_Additive
    translationSpace=BCS_ParentBoneSpace
    alphaBoolBlend {blendOption=Linear}
    boneToModify {boneName=dyn_cloth_bk_r_01}
    rotation {yaw=5}
   }
   n61 = "Transform (Modify) Bone - Bone: dyn_cloth_r_MSTR"(ComponentPose=@n60, Alpha=1, Translation="0,0,0", Rotation="0,5,0", Scale="1,1,1") {
    rotationMode=BMM_Additive
    rotationSpace=BCS_ParentBoneSpace
    translationMode=BMM_Additive
    translationSpace=BCS_ParentBoneSpace
    alphaBoolBlend {blendOption=Linear}
    boneToModify {boneName=dyn_cloth_r_MSTR}
    rotation {yaw=5}
   }
   n62 = Component To Local(ComponentPose=@n61)
   n63 = Get bIsSkyDiving()
   n64 = RigidBody!(
    ComponentPose=Local To Component(LocalPose=Blend Poses by bool(BlendPose_0=@n62, BlendPose_1=@n32, bActiveValue=@n63) {blendType=Linear})
    Alpha:float=0.7
    OverrideWorldGravity="0,0,-980"
    ExternalForce
   ) -> (Pose) {
    bClampLinearTranslationLimitToRefPose=true
    bOverrideWorldGravity=true
    overridePhysicsAsset=&/M_MED_The_Golden_Skeleton_Dyn_torso1
    simulationSpace=BaseBoneSpace
    alphaBoolBlend {blendOption=Linear}
    baseBoneRef {boneName=spine_05}
    componentAppliedLinearAccClamp=<3000 3000 3000>
    componentLinearAccScale=<0.4 0.3 0.3>
    componentLinearVelScale=<0.4 0.3 0.3>
    overrideWorldGravity {z=-980}
   }
   n65 = "Constraint - dyn_necklace_l_06 (PP)"!(ComponentPose=@n64, Alpha=0.7, ConstraintWeights_0=0.9, ConstraintWeights_1:float=0.1) -> (Pose) {
    alphaBoolBlend {blendOption=Linear}
    boneToModify {boneName=dyn_necklace_l_06}
    constraintSetup [{targetBone {boneName=dyn_necklace_l_05}}, {targetBone {boneName=dyn_necklace_03}}]
    constraintWeights [1,1]
   }
   n66 = "Constraint - dyn_necklace_l_04 (PP)"!(ComponentPose=@n65, Alpha=0.7, ConstraintWeights_0=0.5, ConstraintWeights_1:float=0.5) -> (Pose) {
    alphaBoolBlend {blendOption=Linear}
    boneToModify {boneName=dyn_necklace_l_04}
    constraintSetup [{targetBone {boneName=dyn_necklace_l_03}}, {targetBone {boneName=dyn_necklace_02}}]
    constraintWeights [1,1]
   }
   n67 = "Constraint - dyn_necklace_l_05 (PP)"!(ComponentPose=@n66, Alpha=0.7, ConstraintWeights_0=0.5, ConstraintWeights_1:float=0.5) -> (Pose) {
    alphaBoolBlend {blendOption=Linear}
    boneToModify {boneName=dyn_necklace_l_05}
    constraintSetup [{targetBone {boneName=dyn_necklace_l_04}}, {targetBone {boneName=dyn_necklace_l_06}}]
    constraintWeights [1,1]
   }
   n68 = "Constraint - dyn_necklace_l_03 (PP)"!(ComponentPose=@n67, Alpha=0.7, ConstraintWeights_0=0.5, ConstraintWeights_1:float=0.5) -> (Pose) {
    alphaBoolBlend {blendOption=Linear}
    boneToModify {boneName=dyn_necklace_l_03}
    constraintSetup [{targetBone {boneName=dyn_necklace_l_02}}, {targetBone {boneName=dyn_necklace_l_04}}]
    constraintWeights [1,1]
   }
   n69 = "Constraint - dyn_necklace_l_01 (PP)"!(ComponentPose=@n68, Alpha=0.7, ConstraintWeights_0=0.5, ConstraintWeights_1:float=0.5) -> (Pose) {
    alphaBoolBlend {blendOption=Linear}
    boneToModify {boneName=dyn_necklace_l_01}
    constraintSetup [{targetBone {boneName=dyn_necklace_01}}, {targetBone {boneName=dyn_necklace_l_02}}]
    constraintWeights [1,1]
   }
   n70 = "Constraint - dyn_necklace_l_02 (PP)"!(ComponentPose=@n69, Alpha=0.7, ConstraintWeights_0=0.5, ConstraintWeights_1:float=0.5) -> (Pose) {
    alphaBoolBlend {blendOption=Linear}
    boneToModify {boneName=dyn_necklace_l_02}
    constraintSetup [{targetBone {boneName=dyn_necklace_l_01}}, {targetBone {boneName=dyn_necklace_l_03}}]
    constraintWeights [1,1]
   }
   n71 = "Constraint - dyn_necklace_r_06 (PP)"!(ComponentPose=@n70, Alpha=0.7, ConstraintWeights_0=0.9, ConstraintWeights_1:float=0.1) -> (Pose) {
    alphaBoolBlend {blendOption=Linear}
    boneToModify {boneName=dyn_necklace_r_06}
    constraintSetup [{targetBone {boneName=dyn_necklace_r_05}}, {targetBone {boneName=dyn_necklace_03}}]
    constraintWeights [1,1]
   }
   n72 = "Constraint - dyn_necklace_r_04 (PP)"!(ComponentPose=@n71, Alpha=0.7, ConstraintWeights_0=0.5, ConstraintWeights_1:float=0.5) -> (Pose) {
    alphaBoolBlend {blendOption=Linear}
    boneToModify {boneName=dyn_necklace_r_04}
    constraintSetup [{targetBone {boneName=dyn_necklace_r_03}}, {targetBone {boneName=dyn_necklace_02}}]
    constraintWeights [1,1]
   }
   n73 = "Constraint - dyn_necklace_r_05 (PP)"!(ComponentPose=@n72, Alpha=0.7, ConstraintWeights_0=0.5, ConstraintWeights_1:float=0.5) -> (Pose) {
    alphaBoolBlend {blendOption=Linear}
    boneToModify {boneName=dyn_necklace_r_05}
    constraintSetup [{targetBone {boneName=dyn_necklace_r_04}}, {targetBone {boneName=dyn_necklace_r_06}}]
    constraintWeights [1,1]
   }
   n74 = "Constraint - dyn_necklace_r_03 (PP)"!(ComponentPose=@n73, Alpha=0.7, ConstraintWeights_0=0.5, ConstraintWeights_1:float=0.5) -> (Pose) {
    alphaBoolBlend {blendOption=Linear}
    boneToModify {boneName=dyn_necklace_r_03}
    constraintSetup [{targetBone {boneName=dyn_necklace_r_02}}, {targetBone {boneName=dyn_necklace_r_04}}]
    constraintWeights [1,1]
   }
   n75 = "Constraint - dyn_necklace_r_01 (PP)"!(ComponentPose=@n74, Alpha=0.7, ConstraintWeights_0=0.5, ConstraintWeights_1:float=0.5) -> (Pose) {
    alphaBoolBlend {blendOption=Linear}
    boneToModify {boneName=dyn_necklace_r_01}
    constraintSetup [{targetBone {boneName=dyn_necklace_01}}, {targetBone {boneName=dyn_necklace_r_02}}]
    constraintWeights [1,1]
   }
   n76 = "Constraint - dyn_necklace_r_02 (PP)"!(ComponentPose=@n75, Alpha=0.7, ConstraintWeights_0=0.5, ConstraintWeights_1:float=0.5) -> (Pose) {
    alphaBoolBlend {blendOption=Linear}
    boneToModify {boneName=dyn_necklace_r_02}
    constraintSetup [{targetBone {boneName=dyn_necklace_r_01}}, {targetBone {boneName=dyn_necklace_r_03}}]
    constraintWeights [1,1]
   }
   n77 = "Constraint - dyn_necklace_bk_04 (PP)"!(ComponentPose=@n76, Alpha=1, ConstraintWeights_0=0.2, ConstraintWeights_1:float=0.8) -> (Pose) {
    alphaBoolBlend {blendOption=Linear}
    boneToModify {boneName=dyn_necklace_bk_04}
    constraintSetup [{targetBone {boneName=dyn_necklace_01}}, {targetBone {boneName=dyn_necklace_bk_03}}]
    constraintWeights [1,1]
   }
   n78 = "Constraint - dyn_necklace_bk_06 (PP)"!(ComponentPose=@n77, Alpha=0.7, ConstraintWeights_0=0.5, ConstraintWeights_1:float=0.5) -> (Pose) {
    alphaBoolBlend {blendOption=Linear}
    boneToModify {boneName=dyn_necklace_bk_06}
    constraintSetup [{targetBone {boneName=dyn_necklace_bk_05}}, {targetBone {boneName=dyn_necklace_pin_r_MSTR}}]
    constraintWeights [1,1]
   }
   n79 = "Constraint - dyn_necklace_bk_03 (PP)"!(ComponentPose=@n78, Alpha=0.7, ConstraintWeights_0=0.5, ConstraintWeights_1:float=0.5) -> (Pose) {
    alphaBoolBlend {blendOption=Linear}
    boneToModify {boneName=dyn_necklace_bk_03}
    constraintSetup [{targetBone {boneName=dyn_necklace_bk_04}}, {targetBone {boneName=dyn_necklace_bk_02}}]
    constraintWeights [1,1]
   }
   n80 = Anim Dynamics - Bone: dyn_belt_high_01(ComponentPose=Local To Component(LocalPose=Component To Local(ComponentPose=@n79)), Alpha=0.8, ExternalForce="0,0,0") {
    angularDampingOverride=1
    angularSpringConstant=5000
    bAngularSpring=true
    bGravityOverrideInSimSpace=true
    bOverrideAngularDamping=true
    bUseGravityOverride=true
    simulationSpace=RootRelative
    alphaBoolBlend {blendOption=Linear}
    boundBone {boneName=dyn_belt_high_01}
    gravityOverride {z=-980}
    physicsBodyDefinitions [{
      sphereCollisionRadius=0
      boundBone {boneName=dyn_belt_high_01}
      boxExtents=<10 10 10>
      constraintSetup {angularLimitsMax=<15 15 15>, angularLimitsMin=<-15 -15 -15>, angularTarget=<1 0 0>}
      localJointOffset=<0 0 0>
     }]
   }
   n81 = Anim Dynamics - Bone: dyn_belt_high_02(ComponentPose=@n80, Alpha=0.8, ExternalForce="0,0,0") {
    angularDampingOverride=1
    angularSpringConstant=5000
    bAngularSpring=true
    bGravityOverrideInSimSpace=true
    bOverrideAngularDamping=true
    bUseGravityOverride=true
    simulationSpace=RootRelative
    alphaBoolBlend {blendOption=Linear}
    boundBone {boneName=dyn_belt_high_02}
    gravityOverride {z=-980}
    physicsBodyDefinitions [{
      sphereCollisionRadius=0
      boundBone {boneName=dyn_belt_high_02}
      boxExtents=<10 10 10>
      constraintSetup {angularLimitsMax=<15 15 15>, angularLimitsMin=<-15 -15 -15>, angularTarget=<1 0 0>}
      localJointOffset=<0 0 0>
     }]
   }
   n82 = Anim Dynamics - Bone: dyn_belt_high_03(ComponentPose=@n81, Alpha=0.8, ExternalForce="0,0,0") {
    angularDampingOverride=1
    angularSpringConstant=5000
    bAngularSpring=true
    bGravityOverrideInSimSpace=true
    bOverrideAngularDamping=true
    bUseGravityOverride=true
    simulationSpace=RootRelative
    alphaBoolBlend {blendOption=Linear}
    boundBone {boneName=dyn_belt_high_03}
    gravityOverride {z=-980}
    physicsBodyDefinitions [{
      sphereCollisionRadius=0
      boundBone {boneName=dyn_belt_high_03}
      boxExtents=<10 10 10>
      constraintSetup {angularLimitsMax=<15 15 15>, angularLimitsMin=<-15 -15 -15>, angularTarget=<1 0 0>}
      localJointOffset=<0 0 0>
     }]
   }
   n83 = Anim Dynamics - Bone: dyn_belt_low_01(ComponentPose=@n82, Alpha=0.8, ExternalForce="0,0,0") {
    angularDampingOverride=1
    angularSpringConstant=5000
    bAngularSpring=true
    bGravityOverrideInSimSpace=true
    bOverrideAngularDamping=true
    bUseGravityOverride=true
    simulationSpace=RootRelative
    alphaBoolBlend {blendOption=Linear}
    boundBone {boneName=dyn_belt_low_01}
    gravityOverride {z=-980}
    physicsBodyDefinitions [{
      sphereCollisionRadius=0
      boundBone {boneName=dyn_belt_low_01}
      boxExtents=<10 10 10>
      constraintSetup {angularLimitsMax=<15 15 15>, angularLimitsMin=<-15 -15 -15>, angularTarget=<1 0 0>}
      localJointOffset=<0 0 0>
     }]
   }
   n84 = Anim Dynamics - Bone: dyn_belt_low_02(ComponentPose=@n83, Alpha=0.8, ExternalForce="0,0,0") {
    angularDampingOverride=1
    angularSpringConstant=5000
    bAngularSpring=true
    bGravityOverrideInSimSpace=true
    bOverrideAngularDamping=true
    bUseGravityOverride=true
    simulationSpace=RootRelative
    alphaBoolBlend {blendOption=Linear}
    boundBone {boneName=dyn_belt_low_02}
    gravityOverride {z=-980}
    physicsBodyDefinitions [{
      sphereCollisionRadius=0
      boundBone {boneName=dyn_belt_low_02}
      boxExtents=<10 10 10>
      constraintSetup {angularLimitsMax=<15 15 15>, angularLimitsMin=<-15 -15 -15>, angularTarget=<1 0 0>}
      localJointOffset=<0 0 0>
     }]
   }
   n85 = Anim Dynamics - Bone: dyn_belt_low_03(ComponentPose=@n84, Alpha=0.8, ExternalForce="0,0,0") {
    angularDampingOverride=1
    angularSpringConstant=5000
    bAngularSpring=true
    bGravityOverrideInSimSpace=true
    bOverrideAngularDamping=true
    bUseGravityOverride=true
    simulationSpace=RootRelative
    alphaBoolBlend {blendOption=Linear}
    boundBone {boneName=dyn_belt_low_03}
    gravityOverride {z=-980}
    physicsBodyDefinitions [{
      sphereCollisionRadius=0
      boundBone {boneName=dyn_belt_low_03}
      boxExtents=<10 10 10>
      constraintSetup {angularLimitsMax=<15 15 15>, angularLimitsMin=<-15 -15 -15>, angularTarget=<1 0 0>}
      localJointOffset=<0 0 0>
     }]
   }
   n86 = Get bIsAccelerating2D()
   n87 = "Transform (Modify) Bone - Bone: dyn_cloth_bk_01"!(ComponentPose=@n85, bAlphaBoolEnabled:bool=@n86, Translation="0,0,0", Rotation="12,0,0", Scale="1,1,1") -> (Pose) {
    alphaInputType=Bool
    rotationMode=BMM_Additive
    rotationSpace=BCS_BoneSpace
    scaleMode=BMM_Additive
    scaleSpace=BCS_BoneSpace
    translationMode=BMM_Additive
    translationSpace=BCS_BoneSpace
    alphaBoolBlend {blendInTime=0.1, blendOption=Linear, blendOutTime=0.1}
    boneToModify {boneName=dyn_cloth_bk_01}
    rotation {pitch=12}
   }
   n88 = "Transform (Modify) Bone - Bone: dyn_cloth_fr_l_MSTR"!(ComponentPose=@n87, bAlphaBoolEnabled:bool=@n86, Translation="0,0,0", Rotation="12,0,0", Scale="1,1,1") -> (Pose) {
    alphaInputType=Bool
    rotationMode=BMM_Additive
    rotationSpace=BCS_BoneSpace
    scaleMode=BMM_Additive
    scaleSpace=BCS_BoneSpace
    translationMode=BMM_Additive
    translationSpace=BCS_BoneSpace
    alphaBoolBlend {blendInTime=0.1, blendOption=Linear, blendOutTime=0.1}
    boneToModify {boneName=dyn_cloth_fr_l_MSTR}
    rotation {pitch=12}
   }
   n89 = "Transform (Modify) Bone - Bone: dyn_cloth_l_MSTR"!(ComponentPose=@n88, bAlphaBoolEnabled:bool=@n86, Translation="0,0,0", Rotation="12,0,0", Scale="1,1,1") -> (Pose) {
    alphaInputType=Bool
    rotationMode=BMM_Additive
    rotationSpace=BCS_BoneSpace
    scaleMode=BMM_Additive
    scaleSpace=BCS_BoneSpace
    translationMode=BMM_Additive
    translationSpace=BCS_BoneSpace
    alphaBoolBlend {blendInTime=0.1, blendOption=Linear, blendOutTime=0.1}
    boneToModify {boneName=dyn_cloth_l_MSTR}
    rotation {pitch=12}
   }
   n90 = "Transform (Modify) Bone - Bone: dyn_cloth_bk_l_01"!(ComponentPose=@n89, bAlphaBoolEnabled:bool=@n86, Translation="0,0,0", Rotation="12,0,0", Scale="1,1,1") -> (Pose) {
    alphaInputType=Bool
    rotationMode=BMM_Additive
    rotationSpace=BCS_BoneSpace
    scaleMode=BMM_Additive
    scaleSpace=BCS_BoneSpace
    translationMode=BMM_Additive
    translationSpace=BCS_BoneSpace
    alphaBoolBlend {blendInTime=0.1, blendOption=Linear, blendOutTime=0.1}
    boneToModify {boneName=dyn_cloth_bk_l_01}
    rotation {pitch=12}
   }
   n91 = Blend Poses by bool(BlendPose_0=Component To Local(ComponentPose=@n5), BlendPose_1=Component To Local(ComponentPose=@n90), bActiveValue=Get bIsPlayingEmote()) {blendType=Linear}
   Output Pose(Result=Component To Local(ComponentPose=Local To Component(LocalPose=@n91))) {layerGroup=None}
   n92 = "Transform (Modify) Bone - Bone: hand_ulnarad_l"(
    ComponentPose=Local To Component(LocalPose=Copy Pose From Mesh() {bUseAttachedParent=true})
    Alpha=1
    Translation="0,0,0"
    Rotation=Get WristLeftRotation() -> (WristLeftRotation:Rotator)
    Scale="1,1,1"
   ) {
    rotationMode=BMM_Replace
    rotationSpace=BCS_BoneSpace
    translationSpace=BCS_BoneSpace
    alphaBoolBlend {blendOption=Linear}
    boneToModify {boneName=hand_ulnarad_l}
   }
   n93 = "Transform (Modify) Bone - Bone: hand_ulnarad_r"(ComponentPose=@n92, Alpha=1, Translation="0,0,0", Rotation=Get WristRightRotation() -> (WristRightRotation:Rotator), Scale="1,1,1") {
    rotationMode=BMM_Replace
    rotationSpace=BCS_BoneSpace
    translationSpace=BCS_BoneSpace
    alphaBoolBlend {blendOption=Linear}
    boneToModify {boneName=hand_ulnarad_r}
   }
   n94 = "Transform (Modify) Bone - Bone: dyn_crouch_collision"(ComponentPose=@n93, Alpha=Get CrouchInverseAlpha() -> (CrouchInverseAlpha:float), Translation="0,50,0", Rotation="0,0,0", Scale="1,1,1") {
    translationMode=BMM_Replace
    translationSpace=BCS_BoneSpace
    alphaBoolBlend {blendOption=Linear}
    alphaScaleBiasClamp {bInterpResult=true}
    boneToModify {boneName=dyn_crouch_collision}
    translation {y=50}
   }
   n95 = "Constraint - dyn_shoulder_l_MSTR (PPPP)"!(
    ComponentPose=@n94
    Alpha=1
    ConstraintWeights_0=0.0428
    ConstraintWeights_1:float=0.1936
    ConstraintWeights_2:float=0.0882
    ConstraintWeights_3:float=0.6754
   ) -> (Pose) {
    alphaBoolBlend {blendOption=Linear}
    boneToModify {boneName=dyn_shoulder_l_MSTR}
    constraintSetup [
     {targetBone {boneName=spine_03}}
     {targetBone {boneName=spine_05}}
     {targetBone {boneName=upperarm_twist_01_l}}
     {targetBone {boneName=clavicle_l}}
    ]
    constraintWeights [1,1,1,1]
   }
   n96 = "Constraint - dyn_shoulder_r_MSTR (PPPP)"!(
    ComponentPose=@n95
    Alpha=1
    ConstraintWeights_0=0.0882
    ConstraintWeights_1:float=0.6754
    ConstraintWeights_2:float=0.0428
    ConstraintWeights_3:float=0.1936
   ) -> (Pose) {
    alphaBoolBlend {blendOption=Linear}
    boneToModify {boneName=dyn_shoulder_r_MSTR}
    constraintSetup [
     {targetBone {boneName=upperarm_twist_01_r}}
     {targetBone {boneName=clavicle_r}}
     {targetBone {boneName=spine_03}}
     {targetBone {boneName=spine_05}}
    ]
    constraintWeights [1,1,1,1]
   }
   n97 = "Constraint - dyn_necklace_pin_l_MSTR (PPP)"!(ComponentPose=@n96, Alpha=1, ConstraintWeights_0=0.0304, ConstraintWeights_1:float=0.7208, ConstraintWeights_2:float=0.2488) -> (Pose) {
    alphaBoolBlend {blendOption=Linear}
    boneToModify {boneName=dyn_necklace_pin_l_MSTR}
    constraintSetup [{targetBone {boneName=spine_03}}, {targetBone {boneName=spine_05}}, {targetBone {boneName=clavicle_l}}]
    constraintWeights [1,1,1]
   }
   n98 = "Constraint - dyn_necklace_pin_r_MSTR (PPPPP)"!(
    ComponentPose=@n97
    Alpha=1
    ConstraintWeights_0=0.2212
    ConstraintWeights_1:float=0.021
    ConstraintWeights_2:float=0.0062
    ConstraintWeights_3:float=0.0267
    ConstraintWeights_4:float=0.7249
   ) -> (Pose) {
    alphaBoolBlend {blendOption=Linear}
    boneToModify {boneName=dyn_necklace_pin_r_MSTR}
    constraintSetup [
     {targetBone {boneName=clavicle_r}}
     {targetBone {boneName=neck_02}}
     {targetBone {boneName=neck_01}}
     {targetBone {boneName=spine_03}}
     {targetBone {boneName=spine_05}}
    ]
    constraintWeights [1,1,1,1,1]
   }
   n99 = "Transform (Modify) Bone - Bone: dyn_necklace_pin_l_MSTR"(ComponentPose=@n98, Alpha=Get ArmLeftUpAlpha() -> (ArmLeftUpAlpha:float), Translation="0.000,1.670,0.000", Rotation="0,0,0", Scale="1,1,1") {
    rotationMode=BMM_Replace
    rotationSpace=BCS_BoneSpace
    translationMode=BMM_Replace
    translationSpace=BCS_BoneSpace
    alphaBoolBlend {blendOption=Linear}
    alphaScaleBiasClamp {bInterpResult=true, interpSpeedDecreasing=15, interpSpeedIncreasing=15}
    boneToModify {boneName=dyn_necklace_pin_l_MSTR}
    translation {y=1.67}
   }
   n100 = "Transform (Modify) Bone - Bone: dyn_necklace_pin_r_MSTR"(
    ComponentPose=@n99
    Alpha=Get ArmRightUpAlpha() -> (ArmRightUpAlpha:float)
    Translation="-0.000,-1.942,-0.000"
    Rotation="0,0,0"
    Scale="1,1,1"
   ) {
    rotationMode=BMM_Replace
    rotationSpace=BCS_BoneSpace
    translationMode=BMM_Replace
    translationSpace=BCS_BoneSpace
    alphaBoolBlend {blendOption=Linear}
    alphaScaleBiasClamp {bInterpResult=true, interpSpeedDecreasing=15, interpSpeedIncreasing=15}
    boneToModify {boneName=dyn_necklace_pin_r_MSTR}
    translation {y=-1.942}
   }
   n101 = "Transform (Modify) Bone - Bone: dyn_necklace_pin_r_MSTR"(ComponentPose=@n100, Alpha=Get HeadUpAlpha() -> (HeadUpAlpha:float), Translation="-3,0,0", Rotation="0,0,0", Scale="1,1,1") {
    rotationMode=BMM_Replace
    rotationSpace=BCS_BoneSpace
    translationMode=BMM_Replace
    translationSpace=BCS_BoneSpace
    alphaBoolBlend {blendOption=Linear}
    alphaScaleBiasClamp {bInterpResult=true, interpSpeedDecreasing=15, interpSpeedIncreasing=15}
    boneToModify {boneName=dyn_necklace_pin_r_MSTR}
    translation {x=-3}
   }
   n102 = Get CrouchAlpha() -> (CrouchAlpha:float)
   n103 = "Transform (Modify) Bone - Bone: dyn_belt_low_MSTR"(ComponentPose=@n101, Alpha=@n102, Translation="-4.029,-2.198,-1.634", Rotation="-30.013,-34.551,-14.329", Scale="1,1,1") {
    rotationMode=BMM_Replace
    rotationSpace=BCS_BoneSpace
    translationMode=BMM_Replace
    translationSpace=BCS_BoneSpace
    alphaBoolBlend {blendOption=Linear}
    alphaScaleBiasClamp {bInterpResult=true, interpSpeedDecreasing=15, interpSpeedIncreasing=15}
    boneToModify {boneName=dyn_belt_low_MSTR}
    rotation=<rot -30.013 -34.551 -14.329>
    translation=<-4.029 -2.198 -1.634>
   }
   n104 = "Transform (Modify) Bone - Bone: dyn_belt_high_MSTR"(ComponentPose=@n103, Alpha=@n102, Translation="-1.080,0.496,2.147", Rotation="-9.193,-16.512,2.710", Scale="1,1,1") {
    rotationMode=BMM_Replace
    rotationSpace=BCS_BoneSpace
    translationMode=BMM_Replace
    translationSpace=BCS_BoneSpace
    alphaBoolBlend {blendOption=Linear}
    alphaScaleBiasClamp {bInterpResult=true, interpSpeedDecreasing=15, interpSpeedIncreasing=15}
    boneToModify {boneName=dyn_belt_high_MSTR}
    rotation=<rot -9.193 -16.5123 2.71>
    translation=<-1.08 0.496 2.147>
   }
   n105 = "Transform (Modify) Bone - Bone: dyn_cloth_l_MSTR"!(ComponentPose=@n104, Alpha:float=@n102, Translation="-1.695,-4.176,2.627", Rotation="0,0,0", Scale="1,1,1") -> (Pose) {
    rotationMode=BMM_Replace
    rotationSpace=BCS_BoneSpace
    translationMode=BMM_Replace
    translationSpace=BCS_BoneSpace
    alphaBoolBlend {blendOption=Linear}
    alphaScaleBiasClamp {bInterpResult=true, interpSpeedDecreasing=15, interpSpeedIncreasing=15}
    boneToModify {boneName=dyn_cloth_l_MSTR}
    translation=<-1.695 -4.176 2.627>
   }
   n106 = "Transform (Modify) Bone - Bone: dyn_cloth_fr_l_MSTR"!(ComponentPose=@n105, Alpha:float=@n102, Translation="-3.666,7.387,-0.000", Rotation="0,0,0", Scale="1,1,1") -> (Pose) {
    rotationMode=BMM_Replace
    rotationSpace=BCS_BoneSpace
    translationMode=BMM_Replace
    translationSpace=BCS_BoneSpace
    alphaBoolBlend {blendOption=Linear}
    alphaScaleBiasClamp {bInterpResult=true, interpSpeedDecreasing=15, interpSpeedIncreasing=15}
    boneToModify {boneName=dyn_cloth_fr_l_MSTR}
    translation=<-3.666 7.387>
   }
   n107 = "Transform (Modify) Bone - Bone: dyn_cloth_fr_r_MSTR"(ComponentPose=@n106, Alpha=@n102, Translation="-1.533,-5.379,1.638", Rotation="0,0,0", Scale="1,1,1") {
    rotationMode=BMM_Replace
    rotationSpace=BCS_BoneSpace
    translationMode=BMM_Replace
    translationSpace=BCS_BoneSpace
    alphaBoolBlend {blendOption=Linear}
    alphaScaleBiasClamp {bInterpResult=true, interpSpeedDecreasing=15, interpSpeedIncreasing=15}
    boneToModify {boneName=dyn_cloth_fr_r_MSTR}
    translation=<-1.533 -5.379 1.638>
   }
   n108 = "Transform (Modify) Bone - Bone: dyn_bracelet_r"(ComponentPose=@n107, Alpha=Get gauntlet() -> (gauntlet:double), Translation="0,0,0", Rotation="0,0,0", Scale="0,0,0") {scaleMode=BMM_Replace, alphaBoolBlend {blendOption=Linear}, boneToModify {boneName=dyn_bracelet_r}, scale=<0 0 0>}
   Save cached pose 'base Pose'(Pose=Component To Local(ComponentPose=@n108)) {cacheName=base Pose}
   n109 = Get bIsSkyDiving()
   SelectVector(A="10,0,100", B=Get GravityOverride(), bPickA=@n109)
   SelectFloat(A=0.1, B=0.7, bPickA=Get bIsBackpedaling() -> (bIsBackpedaling:bool))
   OR Boolean!(A=Get bIsSkydiveFloating(), B=@n109, C:bool) -> (ReturnValue)
   Get bIsSkydiveDiveMode()
  ]}
 EventGraph {nodes [
   ResetDynamics(InTeleportType=TeleportPhysics)
   > ResetDynamics(InTeleportType=ResetPhysics)
   Event OnWeaponChanged()
   > n1 = Sequence!(execute) -> (then_0, then_1, then_2:Exec, then_3:Exec, then_4:Exec, then_5:Exec)
   n2 = Cast To FortPawn!(execute=@n1.then_1, Object=GetOwningActor()) -> (then, CastFailed, AsFort Pawn:Fort Pawn Object Reference) {is=title K2Node_DynamicCast}
   n3 = Get CurrentWeapon!(self:Fort Pawn Object Reference=@n2) -> (CurrentWeapon:Fort Weapon Object Reference)
   > Set gauntlet(gauntlet:double="To Float (Boolean)"(InBool=WeaponDataHasAnyMatchingGameplayTags!(
      self:Fort Weapon Object Reference=@n3
      TagContainer:Gameplay Tag Container Structure=(GameplayTags=((TagName="Animation.CharacterParts.HideFX.Hands")))
     ) -> (ReturnValue:bool))) -> (Output_Get:double)
   Set in_frontend_chain(execute=@n1.then_2, in_frontend_chain:double=SelectFloat(A=0.7, B=1, bPickA=Get bIsInFrontEnd())) -> (Output_Get:double)
   > Set in_frontend_chain2(in_frontend_chain2:double=SelectFloat(A=0.35, B=1, bPickA=Get bIsInFrontEnd())) -> (Output_Get:double)
   n4 = Branch(execute=@n1.then_3, Condition=Get bIsSkyDiving())
   > Set Is_Sky_Diving_EG(Is_Sky_Diving_EG:bool=true) -> (Output_Get:bool)
   Set Is_Sky_Diving_EG(execute=@n4.else, Is_Sky_Diving_EG=false)
   n5 = Branch(execute=@n1.then_4, Condition=Get bIsSkyDiving())
   > n6 = Branch(Condition=Get bIsParachuteOpened())
   Set gravity_override(execute=[@n6.then, @n5.else], gravity_override="0, 0,-980")
   n7 = Branch(execute=@n6.else, Condition="float > float"(A=Get PawnVelocityZ(), B=50) {is=title K2Node_CallFunction})
   > Set gravity_override(gravity_override="0, 0,300")
   Set gravity_override(execute=@n7.else, gravity_override="0, 0,-200")
   Event BlueprintInitializeAnimation()
   > Set Skel_Mesh(Skel_Mesh=GetOwnerSkeletalMeshComponent())
   Event BlueprintUpdateAnimation()
   n8 = Get Skel_Mesh()
   > Set thigh_shape_fn_l(thigh_shape_fn_l:double=GetAlphaFromReferenceTransformDelta(SkeletalMeshComponent=@n8, BoneName=thigh_l, BaseName=pelvis, InputRange=(X=0,Y=-140), TwistAxis="0,0,1")) -> (Output_Get:double)
   > Set thigh_shape_fn_r(thigh_shape_fn_r:double=GetAlphaFromReferenceTransformDelta(SkeletalMeshComponent=@n8, BoneName=thigh_r, BaseName=pelvis, InputRange=(X=0,Y=-140), TwistAxis="0,0,1")) -> (Output_Get:double)
   > n9 = Gravity Override State Function DEPRECATED(CharacterPartAnimInstance=Self-Reference(), GameTimeInSeconds=GetGameTimeInSeconds(), GravityOverrideParams=Get Gravity Override Params())
   > Set gravity_override(gravity_override=@n9)
  ]}
]
variables [
 gauntlet {type=number}
 Gravity Override Params {
  type=GravityOverrideParamsStruct
  default {
   skydive_DivingDown {gravityOverride=<0 0 450>, windAmplitude=<50 50 50>}
   skydive_FloatingDown {gravityOverride=<0 0 450>, windAmplitude=<10 -20 40>, windFrequency=<2 3 3>, windOffset=<0 0 100>}
   skydive_GoingUp {gravityOverride=<0 0 300>, windAmplitude=<50 -100 80>, windFrequency=<0.3 2 3>}
   skydive_Parachuting {gravityOverride=<0 0 -300>, windAmplitude=<20 -30 40>}
  }
 }
 gravity_override {type=Vector}
 in_frontend_chain {type=number}
 in_frontend_chain2 {type=number}
 Is_Accelerating_EG {type=boolean}
 Is_Sky_Diving_EG {type=boolean}
 no_Emote {type=boolean}
 Skel_Mesh {type=SkeletalMeshComponent}
 thigh_shape_fn_l {type=number}
 thigh_shape_fn_r {type=number}
]
