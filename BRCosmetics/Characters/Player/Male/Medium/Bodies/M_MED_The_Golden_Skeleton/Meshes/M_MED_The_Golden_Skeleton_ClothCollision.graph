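# Physics asset for the Golden Skeleton medium male player body, named as the collision set for its cloth simulation.
# Six capsule bodies approximate only the upper torso: four wide ones stack along spine_02 to spine_05, and a thin one lies on each clavicle.
# Constraints chain the spine bodies and attach both clavicles to spine_05, with free angular motion on the clavicle and spine_03-to-spine_02 joints.
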PhysicsAsset /BRCosmetics/Characters/Player/Male/Medium/Bodies/M_MED_The_Golden_Skeleton/Meshes/M_MED_The_Golden_Skeleton_ClothCollision
bodies {
 clavicle_l [{shapeType=Capsule, length=14.9743, radius=1.4166, rotation=<rot 90 0 0>}]
 clavicle_r [{shapeType=Capsule, length=14.9743, radius=1.4166, rotation=<rot 90 0 0>}]
 spine_02 [{shapeType=Capsule, length=4.6613, radius=9.5256, center=<5.3682 1.1217 0>}]
 spine_03 [{shapeType=Capsule, length=2.7957, radius=11.3912, center=<5.4258 1.31 0>}]
 spine_04 [{shapeType=Capsule, length=7.3373, radius=11.3795, center=<3.5671 1.3999 0>}]
 spine_05 [{shapeType=Capsule, length=9.9313, radius=10.7468, center=<-1.167 0.7416 -0.0005>}]
}
constraints [
 {bone1Name=spine_04, bone2Name=spine_03, swing1LimitDegrees=45, swing2LimitDegrees=45, twistLimitDegrees=45}
 {bone1Name=spine_05, bone2Name=spine_04, swing1LimitDegrees=45, swing2LimitDegrees=45, twistLimitDegrees=45}
 {
  bone1Name=spine_03
  bone2Name=spine_02
  swing1LimitDegrees=45
  swing1Motion=Free
  swing2LimitDegrees=45
  swing2Motion=Free
  twistLimitDegrees=45
  twistMotion=Free
 }
 {
  bone1Name=clavicle_l
  bone2Name=spine_05
  swing1LimitDegrees=45
  swing1Motion=Free
  swing2LimitDegrees=45
  swing2Motion=Free
  twistLimitDegrees=45
  twistMotion=Free
 }
 {
  bone1Name=clavicle_r
  bone2Name=spine_05
  swing1LimitDegrees=45
  swing1Motion=Free
  swing2LimitDegrees=45
  swing2Motion=Free
  twistLimitDegrees=45
  twistMotion=Free
 }
]
