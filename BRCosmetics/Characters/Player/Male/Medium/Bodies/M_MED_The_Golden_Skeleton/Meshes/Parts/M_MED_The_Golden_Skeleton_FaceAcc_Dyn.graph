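# Physics asset for the dynamic face accessory straps on the Golden Skeleton, a medium male player body cosmetic.
# Capsule bodies form three constrained chains hanging off the head: one long back-left strap and shorter front-left and front-right straps.
# Front roots are nearly locked and the back root slightly freer, with wider swing further along and tight twist throughout.
# Unconstrained neck and upper spine capsules, plus the head, provide body collision shapes.
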PhysicsAsset /BRCosmetics/Characters/Player/Male/Medium/Bodies/M_MED_The_Golden_Skeleton/Meshes/Parts/M_MED_The_Golden_Skeleton_FaceAcc_Dyn
bodies {
 dyn_strap_bk_l_01 [{shapeType=Capsule, length=5.0203, radius=2.0169, center=<4.0099 -0.0001 0>, rotation=<rot 90 0 0>}]
 dyn_strap_bk_l_02 [{shapeType=Capsule, length=8.2993, radius=1.6532, center=<4.5935 0 0>, rotation=<rot 90 0 0>}]
 dyn_strap_bk_l_03 [{shapeType=Capsule, length=6.4062, radius=1.8704, center=<4.4119 0 0>, rotation=<rot 90 0 0>}]
 dyn_strap_bk_l_04 [{shapeType=Capsule, length=5.1999, radius=1.8372, center=<4.4119 0 0>, rotation=<rot 90 0 0>}]
 dyn_strap_bk_l_06 [{shapeType=Capsule, length=5.1471, radius=1.89, center=<4.4119 0 0>, rotation=<rot 90 0 0>}]
 dyn_strap_bk_l_07 [{shapeType=Capsule, length=1.8243, radius=2.3195, center=<2.6527 0 0>, rotation=<rot 90 0 0>}]
 dyn_strap_fr_l_01 [{shapeType=Capsule, length=7.1625, radius=1.091, center=<4.3902 -0.0001 0>, rotation=<rot 90 0 0>}]
 dyn_strap_fr_l_02 [{shapeType=Capsule, length=6.7257, radius=1.7879, center=<5.3697 0 0.0003>, rotation=<rot 90 0 0>}]
 dyn_strap_fr_l_03 [{shapeType=Capsule, length=6.7229, radius=2.1239, center=<6.7758 -0.0003 0.0002>, rotation=<rot 90 0 0>}]
 dyn_strap_fr_r_01 [{shapeType=Capsule, length=6.6106, radius=1.643, center=<4.3902 -0.0001 0>, rotation=<rot 90 0 0>}]
 dyn_strap_fr_r_02 [{shapeType=Capsule, length=7.4359, radius=1.4033, center=<4.1348 -0.0001 0.0003>, rotation=<rot 90 0 0>}]
 dyn_strap_fr_r_03 [{shapeType=Capsule, length=7.0273, radius=1.8196, center=<4.6312 -0.0003 0.0002>, rotation=<rot 90 0 0>}]
 head [{shapeType=Capsule, length=10.6234, radius=7.4815, center=<4.6318 0.6799 0.0001>, rotation=<rot 90 0 0>}]
 neck_02 [{shapeType=Capsule, length=5.7494, radius=7.017, center=<0.6906 -0.2329 0>, rotation=<rot 90 0 7.276>}]
 spine_04 [{shapeType=Capsule, length=25.1456, radius=12.3478, center=<3.3049 0 0>}]
 spine_05 [{shapeType=Capsule, length=28.7867, radius=11.679, center=<6.669 0.6809 0.0637>, rotation=<rot -0.1252 -172.0737 -179.8827>}]
}
constraints [
 {bone1Name=dyn_strap_bk_l_01, bone2Name=head, swing1LimitDegrees=5, swing2LimitDegrees=5, twistLimitDegrees=5}
 {bone1Name=dyn_strap_bk_l_02, bone2Name=dyn_strap_bk_l_01, swing1LimitDegrees=15, swing2LimitDegrees=15, twistLimitDegrees=5}
 {bone1Name=dyn_strap_bk_l_03, bone2Name=dyn_strap_bk_l_02, swing1LimitDegrees=15, swing2LimitDegrees=15, twistLimitDegrees=5}
 {bone1Name=dyn_strap_bk_l_04, bone2Name=dyn_strap_bk_l_03, swing1LimitDegrees=30, swing2LimitDegrees=30, twistLimitDegrees=5}
 {bone1Name=dyn_strap_bk_l_06, bone2Name=dyn_strap_bk_l_04, swing1LimitDegrees=30, swing2LimitDegrees=30, twistLimitDegrees=5}
 {bone1Name=dyn_strap_bk_l_07, bone2Name=dyn_strap_bk_l_06, swing1LimitDegrees=30, swing2LimitDegrees=30, twistLimitDegrees=5}
 {bone1Name=dyn_strap_fr_l_01, bone2Name=head, swing1LimitDegrees=0.1, swing2LimitDegrees=0.1, twistLimitDegrees=0.1}
 {bone1Name=dyn_strap_fr_l_02, bone2Name=dyn_strap_fr_l_01, swing1LimitDegrees=15, swing2LimitDegrees=15, twistLimitDegrees=1}
 {bone1Name=dyn_strap_fr_l_03, bone2Name=dyn_strap_fr_l_02, swing1LimitDegrees=15, swing2LimitDegrees=15, twistLimitDegrees=1}
 {bone1Name=dyn_strap_fr_r_01, bone2Name=head, swing1LimitDegrees=0.1, swing2LimitDegrees=0.1, twistLimitDegrees=0.1}
 {bone1Name=dyn_strap_fr_r_02, bone2Name=dyn_strap_fr_r_01, swing1LimitDegrees=30, swing2LimitDegrees=30, twistLimitDegrees=1}
 {bone1Name=dyn_strap_fr_r_03, bone2Name=dyn_strap_fr_r_02, swing1LimitDegrees=30, swing2LimitDegrees=30, twistLimitDegrees=1}
]
